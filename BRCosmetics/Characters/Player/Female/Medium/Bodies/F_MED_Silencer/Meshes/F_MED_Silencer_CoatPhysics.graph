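# Physics asset for the coat of the medium female Silencer character, setting up collision and joints for its dangling jacket bones.
# Left and right back, front and side chains each hang from tiny master bodies on the pelvis with free rotation, then run through five segments with limited swing and mostly tight twist.
# Pelvis and leg shapes provide body colliders, while a floor body uses a large offset sphere, apparently as a ground stand-in.
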PhysicsAsset /BRCosmetics/Characters/Player/Female/Medium/Bodies/F_MED_Silencer/Meshes/F_MED_Silencer_CoatPhysics
bodies {
 calf_l [{shapeType=Capsule, length=45.7953, radius=8.9887, center=<-18.227 -2.3251 0.6745>, rotation=<rot 90 180 1.4636>}]
 calf_r [{shapeType=Capsule, length=62.2764, radius=8.2852, center=<18.227 2.3251 -0.6745>, rotation=<rot 90 -179.7762 -178.3126>}]
 dyn_jacketBack_01_l [{shapeType=Box, extentX=10.5, extentY=19.5, extentZ=4, shapeName=dyn_jacketBack_01_l_box, center=<5 0 4>}]
 dyn_jacketBack_01_r [{shapeType=Box, extentX=10.5, extentY=19.5, extentZ=4, shapeName=dyn_jacketBack_01_r_box, center=<5 0 4>}]
 dyn_jacketBack_02_l [{shapeType=Box, extentX=10.5, extentY=19.5, extentZ=4, shapeName=dyn_jacketBack_02_l_box, center=<5 0 4>}]
 dyn_jacketBack_02_r [{shapeType=Box, extentX=10.5, extentY=19.5, extentZ=4, shapeName=dyn_jacketBack_02_r_box, center=<5 0 4>}]
 dyn_jacketBack_03_l [{shapeType=Capsule, length=20.75, radius=5.25, center=<5 0 6>, rotation=<rot 0 0 90>}]
 dyn_jacketBack_03_r [{shapeType=Capsule, length=20.75, radius=5.25, center=<5 0 6>, rotation=<rot 0 0 90>}]
 dyn_jacketBack_04_l [{shapeType=Capsule, length=32.618, radius=8, center=<5 0 6>, rotation=<rot 0 0 90>}]
 dyn_jacketBack_04_r [{shapeType=Capsule, length=37.4321, radius=8, center=<5 0 6>, rotation=<rot 0 0 90>}]
 dyn_jacketBack_05_l [{shapeType=Capsule, length=32.4877, radius=8, center=<5 0 6>, rotation=<rot 0 0 90>}]
 dyn_jacketBack_05_r [{shapeType=Capsule, length=48.2956, radius=10, center=<5 0 6>, rotation=<rot 0 0 90>}]
 dyn_jacketBack_mstr_l [{shapeType=Capsule, length=0.505, radius=0.505, rotation=<rot 0 -90 0>}]
 dyn_jacketBack_mstr_r [{shapeType=Capsule, length=0.505, radius=0.505, rotation=<rot 0 -90 0>}]
 dyn_jacketFront_01_l [{shapeType=Box, extentX=10, extentY=13.5, extentZ=2, shapeName=dyn_jacketFront_01_l_box, center=<5 0 3>}]
 dyn_jacketFront_01_r [{shapeType=Box, extentX=10, extentY=13.5, extentZ=2, shapeName=dyn_jacketFront_01_r_box, center=<5 0 3>}]
 dyn_jacketFront_02_l [{shapeType=Box, extentX=10, extentY=13.5, extentZ=2, shapeName=dyn_jacketFront_02_l_box, center=<5 0 3>}]
 dyn_jacketFront_02_r [{shapeType=Box, extentX=10, extentY=13.5, extentZ=2, shapeName=dyn_jacketFront_02_r_box, center=<5 0 3>}]
 dyn_jacketFront_03_l [{shapeType=Capsule, length=20, radius=7, center=<5 0 7>, rotation=<rot 0 0 90>}]
 dyn_jacketFront_03_r [{shapeType=Capsule, length=20, radius=7, center=<5 0 7>, rotation=<rot 0 0 79.9995>}]
 dyn_jacketFront_04_l [{shapeType=Capsule, length=20, radius=7, center=<5 0 6>, rotation=<rot 0 0 85.7763>}]
 dyn_jacketFront_04_r [{shapeType=Capsule, length=20, radius=7, center=<5 0 6>, rotation=<rot 0 0 85.7763>}]
 dyn_jacketFront_05_l [{shapeType=Capsule, length=20, radius=7, center=<3 0 6>, rotation=<rot 0 0 85.7763>}]
 dyn_jacketFront_05_r [{shapeType=Capsule, length=38.25, radius=10, center=<3 0 4>, rotation=<rot 0 0 85>}]
 dyn_jacketFront_mstr_l [{shapeType=Capsule, length=0.505, radius=0.505, center=<0 4.9921 0>, rotation=<rot 0 -90 0>}]
 dyn_jacketFront_mstr_r [{shapeType=Capsule, length=0.505, radius=0.505, rotation=<rot 0 -90 0>}]
 dyn_jacketSide_01_l [{shapeType=Capsule, length=20, radius=4.5, center=<5 0 5>, rotation=<rot 0 0 80>}]
 dyn_jacketSide_01_r [{shapeType=Capsule, length=20, radius=4, center=<5 0 5>, rotation=<rot 0 0 100>}]
 dyn_jacketSide_02_l [{shapeType=Capsule, length=20.5, radius=3.5, center=<5 0 5>, rotation=<rot 0 0 80>}]
 dyn_jacketSide_02_r [{shapeType=Capsule, length=20.25, radius=6, center=<5 0 5>, rotation=<rot 0 0 100>}]
 dyn_jacketSide_03_l [{shapeType=Capsule, length=20, radius=5.5, center=<5 0 5>, rotation=<rot 0 0 80>}]
 dyn_jacketSide_03_r [{shapeType=Capsule, length=20, radius=6, center=<5 0 5>, rotation=<rot 0 0 100>}]
 dyn_jacketSide_04_l [{shapeType=Capsule, length=21, radius=10, center=<5 0 6>, rotation=<rot 0 0 80>}]
 dyn_jacketSide_04_r [{shapeType=Capsule, length=21, radius=10, center=<5 0 6>, rotation=<rot 0 0 100>}]
 dyn_jacketSide_05_l [{shapeType=Capsule, length=21, radius=10, center=<5 0 6>, rotation=<rot 0 0 80>}]
 dyn_jacketSide_05_r [{shapeType=Capsule, length=24.25, radius=12, center=<5 0 4>, rotation=<rot 0 0 100>}]
 dyn_jacketSide_mstr_l [{shapeType=Capsule, length=0.505, radius=0.505, center=<0 4.9921 0>, rotation=<rot 0 -90 0>}]
 dyn_jacketSide_mstr_r [{shapeType=Capsule, length=0.505, radius=0.505, center=<0 4.9921 0>, rotation=<rot 0 -90 0>}]
 floor [
  {shapeType=Sphere, radius=200, shapeName=floor_sphere, center=<-3.4923 -2.714 -191.5365>}
  {shapeType=Sphere, radius=22.9194, shapeName=floor_sphere, center=<0 0.5113 -6.6497>}
 ]
 foot_l [{shapeType=Capsule, length=15.6794, radius=7.0209, center=<-3.5001 8.3295 0.0009>, rotation=<rot 0 0 -90>}]
 foot_r [{shapeType=Capsule, length=29.2126, radius=6.7153, center=<6.4296 -6.642 -0.0009>, rotation=<rot 0 0 90>}]
 pelvis [
  {shapeType=Sphere, radius=16.351, shapeName=pelvis_sphere, center=<-8.5464 0.6339 0>}
  {shapeType=Capsule, length=26.8862, radius=12, center=<0.96 -1.1447 0.2284>, rotation=<rot 90 180 171.5626>}
 ]
 thigh_l [
  {shapeType=Capsule, length=33.9498, radius=6.4712, center=<-23.697 -0.2 0.3015>, rotation=<rot 86.4154 0 177.6156>}
  {
   shapeType=Capsule
   length=24.9005
   radius=10.3613
   shapeName=thigh_l_capsule
   center=<-11.9227 0.0575 0.8477>
   rotation=<rot -87.1722 5.7146 -5.9387>
  }
 ]
 thigh_r [
  {shapeType=Capsule, length=33.4703, radius=6.9507, center=<23.697 0.2 -0.3015>, rotation=<rot 86.4154 0 177.6156>}
  {
   shapeType=Capsule
   length=27.1577
   radius=11.4282
   shapeName=thigh_r_capsule
   center=<11.9227 -0.0575 -0.8477>
   rotation=<rot 87.1762 174.5579 174.8909>
  }
 ]
}
constraints [
 {
  bone1Name=dyn_jacketBack_mstr_l
  bone2Name=pelvis
  swing1LimitDegrees=45
  swing1Motion=Free
  swing2LimitDegrees=45
  swing2Motion=Free
  twistLimitDegrees=45
  twistMotion=Free
 }
 {bone1Name=dyn_jacketBack_01_l, bone2Name=dyn_jacketBack_mstr_l, swing1LimitDegrees=10, swing2LimitDegrees=25, twistLimitDegrees=2}
 {bone1Name=dyn_jacketBack_02_l, bone2Name=dyn_jacketBack_01_l, swing1LimitDegrees=10, swing2LimitDegrees=25, twistLimitDegrees=2}
 {bone1Name=dyn_jacketBack_03_l, bone2Name=dyn_jacketBack_02_l, swing1LimitDegrees=10, swing2LimitDegrees=45, twistLimitDegrees=2}
 {bone1Name=dyn_jacketBack_04_l, bone2Name=dyn_jacketBack_03_l, swing1LimitDegrees=10, swing2LimitDegrees=45, twistLimitDegrees=2}
 {
  bone1Name=dyn_jacketBack_mstr_r
  bone2Name=pelvis
  swing1LimitDegrees=45
  swing1Motion=Free
  swing2LimitDegrees=45
  swing2Motion=Free
  twistLimitDegrees=45
  twistMotion=Free
 }
 {bone1Name=dyn_jacketBack_01_r, bone2Name=dyn_jacketBack_mstr_r, swing1LimitDegrees=10, swing2LimitDegrees=25, twistLimitDegrees=2}
 {bone1Name=dyn_jacketBack_02_r, bone2Name=dyn_jacketBack_01_r, swing1LimitDegrees=10, swing2LimitDegrees=25, twistLimitDegrees=2}
 {bone1Name=dyn_jacketBack_03_r, bone2Name=dyn_jacketBack_02_r, swing1LimitDegrees=10, swing2LimitDegrees=45, twistLimitDegrees=2}
 {bone1Name=dyn_jacketBack_04_r, bone2Name=dyn_jacketBack_03_r, swing1LimitDegrees=10, swing2LimitDegrees=45, twistLimitDegrees=2}
 {
  bone1Name=dyn_jacketFront_mstr_l
  bone2Name=pelvis
  swing1LimitDegrees=45
  swing1Motion=Free
  swing2LimitDegrees=45
  swing2Motion=Free
  twistLimitDegrees=45
  twistMotion=Free
 }
 {bone1Name=dyn_jacketFront_01_l, bone2Name=dyn_jacketFront_mstr_l, swing1LimitDegrees=5, swing2LimitDegrees=45, twistLimitDegrees=2}
 {bone1Name=dyn_jacketFront_02_l, bone2Name=dyn_jacketFront_01_l, swing1LimitDegrees=5, swing2LimitDegrees=60, twistLimitDegrees=2}
 {bone1Name=dyn_jacketFront_03_l, bone2Name=dyn_jacketFront_02_l, swing1LimitDegrees=20, swing2LimitDegrees=30, twistLimitDegrees=2}
 {bone1Name=dyn_jacketFront_04_l, bone2Name=dyn_jacketFront_03_l, swing1LimitDegrees=20, swing2LimitDegrees=30, twistLimitDegrees=2}
 {
  bone1Name=dyn_jacketFront_mstr_r
  bone2Name=pelvis
  swing1LimitDegrees=45
  swing1Motion=Free
  swing2LimitDegrees=45
  swing2Motion=Free
  twistLimitDegrees=45
  twistMotion=Free
 }
 {bone1Name=dyn_jacketFront_01_r, bone2Name=dyn_jacketFront_mstr_r, swing1LimitDegrees=5, swing2LimitDegrees=45, twistLimitDegrees=5}
 {bone1Name=dyn_jacketFront_02_r, bone2Name=dyn_jacketFront_01_r, swing1LimitDegrees=5, swing2LimitDegrees=60, twistLimitDegrees=5}
 {bone1Name=dyn_jacketFront_03_r, bone2Name=dyn_jacketFront_02_r, swing1LimitDegrees=20, swing2LimitDegrees=30, twistLimitDegrees=2}
 {bone1Name=dyn_jacketFront_04_r, bone2Name=dyn_jacketFront_03_r, swing1LimitDegrees=20, swing2LimitDegrees=30, twistLimitDegrees=2}
 {
  bone1Name=dyn_jacketSide_mstr_l
  bone2Name=pelvis
  swing1LimitDegrees=45
  swing1Motion=Free
  swing2LimitDegrees=45
  swing2Motion=Free
  twistLimitDegrees=45
  twistMotion=Free
 }
 {bone1Name=dyn_jacketSide_01_l, bone2Name=dyn_jacketSide_mstr_l, swing1LimitDegrees=10, swing2LimitDegrees=25, twistLimitDegrees=2}
 {bone1Name=dyn_jacketSide_02_l, bone2Name=dyn_jacketSide_01_l, swing1LimitDegrees=10, swing2LimitDegrees=45, twistLimitDegrees=2}
 {bone1Name=dyn_jacketSide_03_l, bone2Name=dyn_jacketSide_02_l, swing1LimitDegrees=10, swing2LimitDegrees=45, twistLimitDegrees=2}
 {bone1Name=dyn_jacketSide_04_l, bone2Name=dyn_jacketSide_03_l, swing1LimitDegrees=10, swing2LimitDegrees=45, twistLimitDegrees=2}
 {
  bone1Name=dyn_jacketSide_mstr_r
  bone2Name=pelvis
  swing1LimitDegrees=45
  swing1Motion=Free
  swing2LimitDegrees=45
  swing2Motion=Free
  twistLimitDegrees=45
  twistMotion=Free
 }
 {bone1Name=dyn_jacketSide_01_r, bone2Name=dyn_jacketSide_mstr_r, swing1LimitDegrees=10, swing2LimitDegrees=45, twistLimitDegrees=2}
 {bone1Name=dyn_jacketSide_02_r, bone2Name=dyn_jacketSide_01_r, swing1LimitDegrees=10, swing2LimitDegrees=45, twistLimitDegrees=2}
 {bone1Name=dyn_jacketSide_03_r, bone2Name=dyn_jacketSide_02_r, swing1LimitDegrees=10, swing2LimitDegrees=45, twistLimitDegrees=2}
 {bone1Name=dyn_jacketSide_04_r, bone2Name=dyn_jacketSide_03_r, swing1LimitDegrees=10, swing2LimitDegrees=45, twistLimitDegrees=2}
 {
  bone1Name=thigh_l
  bone2Name=pelvis
  swing1LimitDegrees=45
  swing1Motion=Free
  swing2LimitDegrees=45
  swing2Motion=Free
  twistLimitDegrees=45
  twistMotion=Free
 }
 {bone1Name=calf_l, bone2Name=thigh_l, swing1LimitDegrees=10, swing2LimitDegrees=45, twistLimitDegrees=2}
 {
  bone1Name=foot_l
  bone2Name=calf_l
  swing1LimitDegrees=45
  swing1Motion=Free
  swing2LimitDegrees=45
  swing2Motion=Free
  twistLimitDegrees=45
  twistMotion=Free
 }
 {
  bone1Name=thigh_l
  bone2Name=thigh_r
  swing1LimitDegrees=45
  swing1Motion=Free
  swing2LimitDegrees=45
  swing2Motion=Free
  twistLimitDegrees=45
  twistMotion=Free
 }
 {
  bone1Name=thigh_r
  bone2Name=pelvis
  swing1LimitDegrees=45
  swing1Motion=Free
  swing2LimitDegrees=45
  swing2Motion=Free
  twistLimitDegrees=45
  twistMotion=Free
 }
 {bone1Name=calf_r, bone2Name=thigh_r, swing1LimitDegrees=10, swing2LimitDegrees=45, twistLimitDegrees=2}
 {
  bone1Name=foot_r
  bone2Name=calf_r
  swing1LimitDegrees=45
  swing1Motion=Free
  swing2LimitDegrees=45
  swing2Motion=Free
  twistLimitDegrees=45
  twistMotion=Free
 }
 {bone1Name=dyn_jacketBack_05_l, bone2Name=dyn_jacketBack_04_l, swing1LimitDegrees=10, swing2LimitDegrees=45, twistLimitDegrees=2}
 {bone1Name=dyn_jacketBack_05_r, bone2Name=dyn_jacketBack_04_r, swing1LimitDegrees=10, swing2LimitDegrees=45, twistLimitDegrees=5}
 {bone1Name=dyn_jacketFront_05_l, bone2Name=dyn_jacketFront_04_l, swing1LimitDegrees=20, swing2LimitDegrees=30, twistLimitDegrees=2}
 {bone1Name=dyn_jacketFront_05_r, bone2Name=dyn_jacketFront_04_r, swing1LimitDegrees=20, swing2LimitDegrees=30, twistLimitDegrees=2}
 {bone1Name=dyn_jacketSide_05_l, bone2Name=dyn_jacketSide_04_l, swing1LimitDegrees=10, swing2LimitDegrees=45, twistLimitDegrees=2}
 {bone1Name=dyn_jacketSide_05_r, bone2Name=dyn_jacketSide_04_r, swing1LimitDegrees=45, swing2LimitDegrees=45, twistLimitDegrees=5}
]
