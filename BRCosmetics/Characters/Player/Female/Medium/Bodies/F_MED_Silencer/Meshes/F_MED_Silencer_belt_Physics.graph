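# Physics asset for the belt mesh of the female medium Silencer player cosmetic, setting up jointed dangling belt and tie chains.
# The belt chain (dyn_belt_02–08) hangs from the pelvis and the tie chain (dyn_tie_01–04) from spine_05, both using capsules with very tight twist limits.
# Pelvis, upper spine, thigh and calf shapes provide body volumes around the chains, with angularly free spine joints and limited leg joints.
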PhysicsAsset /BRCosmetics/Characters/Player/Female/Medium/Bodies/F_MED_Silencer/Meshes/F_MED_Silencer_belt_Physics
bodies {
 calf_l [{shapeType=Capsule, length=33, radius=5.75, center=<-19.5654 1.624 0.7647>, rotation=<rot 0 98.7182 -87.5765>}]
 calf_r [{shapeType=Capsule, length=33, radius=5.75, center=<19.5654 -1.6239 -0.7647>, rotation=<rot 0 94.2665 -88.787>}]
 dyn_belt_02 [{shapeType=Capsule, length=6, radius=1.5, shapeName=dyn_belt_02_capsule, center=<3 1 0>, rotation=<rot 90 0 0>}]
 dyn_belt_03 [{shapeType=Capsule, length=6, radius=1.5, center=<3 1 0>, rotation=<rot 90 0 0>}]
 dyn_belt_04 [{shapeType=Capsule, length=6, radius=1.5, shapeName=dyn_belt_04_capsule, center=<3 1 0>, rotation=<rot 90 0 0>}]
 dyn_belt_05 [{shapeType=Capsule, length=6, radius=1.5, shapeName=dyn_belt_05_capsule, center=<3 1 0>, rotation=<rot 90 0 0>}]
 dyn_belt_06 [{shapeType=Capsule, length=6, radius=1.5, shapeName=dyn_belt_06_capsule, center=<3 1 0>, rotation=<rot 90 0 0>}]
 dyn_belt_07 [{shapeType=Capsule, length=6, radius=1.5, shapeName=dyn_belt_07_capsule, center=<3 1 0>, rotation=<rot 90 0 0>}]
 dyn_belt_08 [{shapeType=Capsule, length=6, radius=1.5, shapeName=dyn_belt_08_capsule, center=<3 1 0>, rotation=<rot 90 0 0>}]
 dyn_tie_01 [{shapeType=Capsule, length=8, radius=3, shapeName=dyn_tie_02_capsule, center=<4 2 0>, rotation=<rot 90 0 0>}]
 dyn_tie_02 [{shapeType=Capsule, length=8, radius=3, shapeName=dyn_tie_02_capsule, center=<4 2 0>, rotation=<rot 90 0 0>}]
 dyn_tie_03 [{shapeType=Capsule, length=8, radius=3, shapeName=dyn_tie_03_capsule, center=<4 2 0>, rotation=<rot 90 159.444 174.4438>}]
 dyn_tie_04 [{shapeType=Capsule, length=11.5804, radius=3.7833, shapeName=dyn_tie_04_capsule, center=<5.4819 1.8329 0>, rotation=<rot 90 -12.1388 0>}]
 pelvis [
  {shapeType=Sphere, radius=11.25, shapeName=pelvis_sphere, center=<11.1374 -3.0856 -4.4612>}
  {shapeType=Capsule, length=24.7374, radius=10.7519, shapeName=pelvis_capsule, center=<-1.5498 -2.7122 1.169>}
 ]
 spine_03 [{shapeType=Capsule, length=27.25, radius=8.25, shapeName=spine_03_capsule, center=<0 1.6312 0>}]
 spine_04 [{shapeType=Capsule, length=27.25, radius=8.25, shapeName=spine_04_capsule, center=<0 2.946 0>}]
 spine_05 [{shapeType=Capsule, length=27.25, radius=8.25, shapeName=spine_05_capsule, center=<0 3.0274 0>}]
 thigh_l [{shapeType=Capsule, length=40.5, radius=8.75, center=<-19.5978 0.0946 1.3934>, rotation=<rot 0 -86.1985 -88.7036>}]
 thigh_r [{shapeType=Capsule, length=29.4123, radius=9.5877, center=<19.5913 0.3283 -1.3934>, rotation=<rot 0 -89.1113 -87.3439>}]
}
constraints [
 {bone1Name=thigh_l, bone2Name=pelvis, swing1LimitDegrees=45, swing2LimitDegrees=45, twistLimitDegrees=45}
 {bone1Name=calf_l, bone2Name=thigh_l, swing1LimitDegrees=45, swing2LimitDegrees=45, twistLimitDegrees=45}
 {bone1Name=thigh_r, bone2Name=pelvis, swing1LimitDegrees=45, swing2LimitDegrees=45, twistLimitDegrees=45}
 {bone1Name=calf_r, bone2Name=thigh_r, swing1LimitDegrees=45, swing2LimitDegrees=45, twistLimitDegrees=45}
 {bone1Name=dyn_belt_03, bone2Name=dyn_belt_02, swing1LimitDegrees=45, swing2LimitDegrees=30, twistLimitDegrees=3}
 {bone1Name=dyn_belt_02, bone2Name=pelvis, swing1LimitDegrees=20, swing2LimitDegrees=20, twistLimitDegrees=2}
 {bone1Name=dyn_belt_04, bone2Name=dyn_belt_03, swing1LimitDegrees=45, swing2LimitDegrees=30, twistLimitDegrees=3}
 {bone1Name=dyn_belt_05, bone2Name=dyn_belt_04, swing1LimitDegrees=45, swing2LimitDegrees=30, twistLimitDegrees=3}
 {bone1Name=dyn_belt_07, bone2Name=dyn_belt_06, swing1LimitDegrees=45, swing2LimitDegrees=30, twistLimitDegrees=3}
 {bone1Name=dyn_belt_06, bone2Name=dyn_belt_05, swing1LimitDegrees=45, swing2LimitDegrees=30, twistLimitDegrees=3}
 {bone1Name=dyn_belt_08, bone2Name=dyn_belt_07, swing1LimitDegrees=45, swing2LimitDegrees=30, twistLimitDegrees=3}
 {
  bone1Name=spine_03
  bone2Name=pelvis
  swing1LimitDegrees=45
  swing1Motion=Free
  swing2LimitDegrees=45
  swing2Motion=Free
  twistLimitDegrees=45
  twistMotion=Free
 }
 {
  bone1Name=spine_04
  bone2Name=spine_03
  swing1LimitDegrees=45
  swing1Motion=Free
  swing2LimitDegrees=45
  swing2Motion=Free
  twistLimitDegrees=45
  twistMotion=Free
 }
 {
  bone1Name=spine_05
  bone2Name=spine_04
  swing1LimitDegrees=45
  swing1Motion=Free
  swing2LimitDegrees=45
  swing2Motion=Free
  twistLimitDegrees=45
  twistMotion=Free
 }
 {bone1Name=dyn_tie_02, bone2Name=dyn_tie_01, swing1LimitDegrees=20, swing2LimitDegrees=20, twistLimitDegrees=1}
 {bone1Name=dyn_tie_03, bone2Name=dyn_tie_02, swing1LimitDegrees=20, swing2LimitDegrees=20, twistLimitDegrees=1}
 {bone1Name=dyn_tie_04, bone2Name=dyn_tie_03, swing1LimitDegrees=20, swing2LimitDegrees=20, twistLimitDegrees=1}
 {bone1Name=dyn_tie_01, bone2Name=spine_05, swing1LimitDegrees=15, swing2LimitDegrees=15, twistLimitDegrees=2}
]
