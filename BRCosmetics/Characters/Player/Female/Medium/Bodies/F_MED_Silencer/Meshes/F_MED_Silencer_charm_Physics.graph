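# Physics asset for the charm mesh of the F_MED_Silencer female player cosmetic, covering a left hip chain and left front jacket piece.
# Capsule bodies on the pelvis, both thighs, four chain bones and one jacket bone approximate the lower body and dangling parts.
# The chain links constrain in series from the pelvis with limited swing and almost no twist.
# The thighs and jacket also constrain to the pelvis, with the jacket's swing and twist left free.
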PhysicsAsset /BRCosmetics/Characters/Player/Female/Medium/Bodies/F_MED_Silencer/Meshes/F_MED_Silencer_charm_Physics
bodies {
 dyn_hipchain_l_01 [{
   shapeType=Capsule
   length=1.005
   radius=1.255
   shapeName=dyn_hipchain_l_01_capsule
   center=<0.3957 0.2496 -0.0617>
   rotation=<rot 0 99.3762 -74.951>
  }]
 dyn_hipchain_l_02 [{
   shapeType=Capsule
   length=1.192
   radius=0.924
   shapeName=dyn_hipchain_l_02_capsule
   center=<0.6685 0.624 0.2257>
   rotation=<rot 0 93.5281 -88.2786>
  }]
 dyn_hipchain_l_03 [{
   shapeType=Capsule
   length=0.9022
   radius=1.1931
   shapeName=dyn_hipchain_l_03_capsule
   center=<0.7611 0.7265 0.3279>
   rotation=<rot 0 87.1329 -83.0111>
  }]
 dyn_hipchain_l_04 [{
   shapeType=Capsule
   length=7.5
   radius=1.5
   shapeName=dyn_hipchain_l_04_capsule
   center=<3.5152 1.255 0.0508>
   rotation=<rot -85.0001 0.0001 5>
  }]
 dyn_jacketFront_02_l [{
   shapeType=Capsule
   length=28.505
   radius=3.755
   shapeName=dyn_jacketFront_02_l_capsule
   center=<0 1.0155 1.6345>
   rotation=<rot -0.0001 -89.9998 89.9999>
  }]
 pelvis [{shapeType=Capsule, length=24.7374, radius=10.7519, shapeName=pelvis_capsule, center=<-1.5498 -2.7122 1.169>}]
 thigh_l [{shapeType=Capsule, length=40.5, radius=8.75, center=<-19.5978 0.0946 1.3934>, rotation=<rot 0 -86.1985 -88.7036>}]
 thigh_r [{shapeType=Capsule, length=29.4123, radius=9.5877, center=<19.5913 0.3283 -1.3934>, rotation=<rot 0 -89.1113 -87.3439>}]
}
constraints [
 {bone1Name=thigh_l, bone2Name=pelvis, swing1LimitDegrees=45, swing2LimitDegrees=45, twistLimitDegrees=45}
 {bone1Name=thigh_r, bone2Name=pelvis, swing1LimitDegrees=45, swing2LimitDegrees=45, twistLimitDegrees=45}
 {bone1Name=dyn_hipchain_l_04, bone2Name=dyn_hipchain_l_03, swing1LimitDegrees=45, swing2LimitDegrees=45, twistLimitDegrees=1}
 {bone1Name=dyn_hipchain_l_02, bone2Name=dyn_hipchain_l_01, swing1LimitDegrees=45, swing2LimitDegrees=45, twistLimitDegrees=1}
 {bone1Name=dyn_hipchain_l_03, bone2Name=dyn_hipchain_l_02, swing1LimitDegrees=45, swing2LimitDegrees=45, twistLimitDegrees=1}
 {bone1Name=dyn_hipchain_l_01, bone2Name=pelvis, swing1LimitDegrees=45, swing2LimitDegrees=45, twistLimitDegrees=1}
 {
  bone1Name=dyn_jacketFront_02_l
  bone2Name=pelvis
  swing1LimitDegrees=45
  swing1Motion=Free
  swing2LimitDegrees=45
  swing2Motion=Free
  twistLimitDegrees=45
  twistMotion=Free
 }
]
