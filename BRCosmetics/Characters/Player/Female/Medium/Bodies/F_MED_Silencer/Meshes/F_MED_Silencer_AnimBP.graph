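# Animation Blueprint for the F_MED_Silencer player body cosmetic, adding secondary motion to its coat, belt and hip chain.
# One StartupCache branch runs RigidBody coat physics with overridden gravity, then corrective bone offsets.
# Another chains Look Ats, pelvis constraints and per-bone AnimDynamics, using leg-driven spherical limits to deflect coat panels.
# A trailCache branch reorients and flattens the belt master bone, then partially constrains coat, belt and chain bones to pelvis and thighs.
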AnimBlueprint /BRCosmetics/Characters/Player/Female/Medium/Bodies/F_MED_Silencer/Meshes/F_MED_Silencer_AnimBP
target_skeleton=&/F_MED_Silencer_Skeleton
graphs [
 AnimGraph {nodes [
   "Transform (Modify) Bone - Bone: dyn_jacketFront_02_r"!(ComponentPose, Alpha=1, Translation="0,0,0") -> (Pose) {
    rotationSpace=BCS_BoneSpace
    translationMode=BMM_Additive
    translationSpace=BCS_BoneSpace
    alphaBoolBlend {blendOption=Linear}
    alphaScaleBiasClamp {bInterpResult=true, interpSpeedDecreasing=5, interpSpeedIncreasing=7}
    boneToModify {boneName=dyn_jacketFront_02_r}
   }
   n1 = RigidBody!(
    ComponentPose=Local To Component(LocalPose=Use cached pose 'StartupCache'())
    Alpha
    OverrideWorldGravity:Vector=Get GravityOverride()
    ExternalForce
   ) -> (Pose) {
    bClampLinearTranslationLimitToRefPose=true
    bOverrideWorldGravity=true
    overridePhysicsAsset=&/F_MED_Silencer_CoatPhysics
    simulationSpace=BaseBoneSpace
    alphaBoolBlend {blendOption=Linear}
    baseBoneRef {boneName=pelvis}
    componentAppliedLinearAccClamp=<3000 3000 3000>
    componentLinearAccScale=<0.5 0.3 0.3>
    componentLinearVelScale=<0.5 0.3 0.3>
    simSpaceSettings {maxAngularVelocity=3, maxLinearAcceleration=700, maxLinearVelocity=500, velocityScaleZ=0, worldAlpha=1, externalLinearDragV=<2 0.2>}
   }
   n2 = "Transform (Modify) Bone - Bone: dyn_jacketBack_04_r"!(ComponentPose=@n1, Alpha=1, Rotation="9,-22,-5") -> (Pose) {
    rotationMode=BMM_Additive
    rotationSpace=BCS_BoneSpace
    alphaBoolBlend {blendOption=Linear}
    alphaScaleBiasClamp {bInterpResult=true, interpSpeedDecreasing=5, interpSpeedIncreasing=7}
    boneToModify {boneName=dyn_jacketBack_04_r}
    rotation=<rot 9 -22 -5>
   }
   n3 = "Transform (Modify) Bone - Bone: dyn_jacketBack_mstr_l"!(ComponentPose=@n2, Alpha=1, Translation="0,0,0.5", Rotation="0,0,0") -> (Pose) {
    rotationMode=BMM_Additive
    rotationSpace=BCS_BoneSpace
    translationMode=BMM_Additive
    translationSpace=BCS_BoneSpace
    alphaBoolBlend {blendOption=Linear}
    alphaScaleBiasClamp {bInterpResult=true, interpSpeedDecreasing=5, interpSpeedIncreasing=7}
    boneToModify {boneName=dyn_jacketBack_mstr_l}
    translation {z=0.5}
   }
   n4 = "Transform (Modify) Bone - Bone: dyn_jacketBack_mstr_r"!(ComponentPose=@n3, Alpha=1, Translation="2,0,0", Rotation="5,0,0") -> (Pose) {
    rotationMode=BMM_Additive
    rotationSpace=BCS_BoneSpace
    translationMode=BMM_Additive
    translationSpace=BCS_BoneSpace
    alphaBoolBlend {blendOption=Linear}
    alphaScaleBiasClamp {bInterpResult=true, interpSpeedDecreasing=5, interpSpeedIncreasing=7}
    boneToModify {boneName=dyn_jacketBack_mstr_r}
    rotation {pitch=5}
    translation {x=2}
   }
   n5 = "Transform (Modify) Bone - Bone: dyn_jacketSide_mstr_r"!(ComponentPose=@n4, Alpha=1, Translation="0,0,0", Rotation="10.000,-0.000,0.000") -> (Pose) {
    rotationMode=BMM_Additive
    rotationSpace=BCS_BoneSpace
    translationMode=BMM_Additive
    translationSpace=BCS_BoneSpace
    alphaBoolBlend {blendOption=Linear}
    alphaScaleBiasClamp {bInterpResult=true, interpSpeedDecreasing=5, interpSpeedIncreasing=7}
    boneToModify {boneName=dyn_jacketSide_mstr_r}
    rotation=<rot 10 0 0>
   }
   n6 = "Transform (Modify) Bone - Bone: dyn_jacketFront_mstr_r"!(ComponentPose=@n5, Alpha=1, Translation="0,0,0", Rotation="10.000,-0.000,0.000") -> (Pose) {
    rotationMode=BMM_Additive
    rotationSpace=BCS_BoneSpace
    translationMode=BMM_Additive
    translationSpace=BCS_BoneSpace
    alphaBoolBlend {blendOption=Linear}
    alphaScaleBiasClamp {bInterpResult=true, interpSpeedDecreasing=5, interpSpeedIncreasing=7}
    boneToModify {boneName=dyn_jacketFront_mstr_r}
    rotation=<rot 10 0 0>
   }
   n7 = "Constraint - dyn_jacketFront_02_r (P)"(ComponentPose=@n6, Alpha=1, ConstraintWeights_0=1) {
    alphaBoolBlend {blendOption=Linear}
    alphaScaleBiasClamp {bClampResult=true, clampMax=0.6}
    boneToModify {boneName=dyn_jacketFront_02_r}
    constraintSetup [{targetBone {boneName=thigh_r}}]
    constraintWeights [1]
   }
   n8 = Look At - Bone: dyn_jacketSide_03_r(ComponentPose=Local To Component(LocalPose=Use cached pose 'StartupCache'()), Alpha=0.5) {
    alphaBoolBlend {blendOption=Linear}
    boneToModify {boneName=dyn_jacketSide_03_r}
    lookAtLocation=<16.6765 43.6858 54.9763>
    lookAtTarget {boneReference {boneName=calf_r}}
    lookAt_Axis {axis=<1 0>}
   }
   n9 = Look At - Bone: dyn_jacketBack_03_r(ComponentPose=@n8, Alpha=0.3) {
    alphaBoolBlend {blendOption=Linear}
    boneToModify {boneName=dyn_jacketBack_03_r}
    lookAtLocation=<20.903 31.2164 67.4651>
    lookAtTarget {boneReference {boneName=calf_r}}
    lookAt_Axis {axis=<1 0>}
   }
   n10 = Look At - Bone: dyn_jacketSide_03_l(ComponentPose=@n9, Alpha=0.5) {
    alphaBoolBlend {blendOption=Linear}
    boneToModify {boneName=dyn_jacketSide_03_l}
    lookAtLocation=<-47.5637 -68.5389 -33.7157>
    lookAtTarget {boneReference {boneName=calf_l}}
    lookAt_Axis {axis=<1 0>}
   }
   n11 = Look At - Bone: dyn_jacketBack_03_l(ComponentPose=@n10, Alpha=0.3) {
    alphaBoolBlend {blendOption=Linear}
    boneToModify {boneName=dyn_jacketBack_03_l}
    lookAtLocation=<-43.9783 -86.5129 -18.7168>
    lookAtTarget {boneReference {boneName=calf_l}}
    lookAt_Axis {axis=<1 0>}
   }
   n12 = Look At - Bone: dyn_jacketSide_01_r(ComponentPose=@n11, Alpha=0.5) {
    alphaBoolBlend {blendOption=Linear}
    boneToModify {boneName=dyn_jacketSide_01_r}
    lookAtLocation=<46.4772 -9.3289 31.0421>
    lookAtTarget {boneReference {boneName=thigh_r}}
    lookAt_Axis {axis=<1 0>}
   }
   n13 = Look At - Bone: dyn_jacketFront_01_r(ComponentPose=@n12, Alpha=0.3) {
    alphaBoolBlend {blendOption=Linear}
    boneToModify {boneName=dyn_jacketFront_01_r}
    lookAtLocation=<5.0821 9.6779 44.778>
    lookAtTarget {boneReference {boneName=thigh_r}}
    lookAt_Axis {axis=<1 0>}
   }
   n14 = Look At - Bone: dyn_jacketSide_01_l(ComponentPose=@n13, Alpha=0.5) {
    alphaBoolBlend {blendOption=Linear}
    boneToModify {boneName=dyn_jacketSide_01_l}
    lookAtLocation=<-98.6603 -198.5858 -102.2869>
    lookAtTarget {boneReference {boneName=thigh_l}}
    lookAt_Axis {axis=<1 0>}
   }
   n15 = Look At - Bone: dyn_jacketFront_01_l(ComponentPose=@n14, Alpha=0.3) {
    alphaBoolBlend {blendOption=Linear}
    boneToModify {boneName=dyn_jacketFront_01_l}
    lookAtLocation=<-76.6609 -14.1213 -90.8918>
    lookAtTarget {boneReference {boneName=thigh_l}}
    lookAt_Axis {axis=<1 0>}
   }
   n16 = "Constraint - dyn_jacketFront_01_l (P)"(ComponentPose=@n15, Alpha=1, ConstraintWeights_0=1) {
    alphaBoolBlend {blendOption=Linear}
    alphaScaleBiasClamp {bClampResult=true, clampMax=0.4}
    boneToModify {boneName=dyn_jacketFront_01_l}
    constraintSetup [{targetBone {boneName=pelvis}}]
    constraintWeights [1]
   }
   n17 = "Constraint - dyn_jacketSide_01_l (P)"(ComponentPose=@n16, Alpha=1, ConstraintWeights_0=1) {
    alphaBoolBlend {blendOption=Linear}
    alphaScaleBiasClamp {bClampResult=true, clampMax=0.4}
    boneToModify {boneName=dyn_jacketSide_01_l}
    constraintSetup [{targetBone {boneName=pelvis}}]
    constraintWeights [1]
   }
   n18 = "Constraint - dyn_jacketBack_01_l (P)"(ComponentPose=@n17, Alpha=1, ConstraintWeights_0=1) {
    alphaBoolBlend {blendOption=Linear}
    alphaScaleBiasClamp {bClampResult=true, clampMax=0.4}
    boneToModify {boneName=dyn_jacketBack_01_l}
    constraintSetup [{targetBone {boneName=pelvis}}]
    constraintWeights [1]
   }
   n19 = Anim Dynamics - Bone: dyn_jacketFront_01_l(ComponentPose=@n18, Alpha=1, ExternalForce="0,0,0") {
    angularDampingOverride=0.9999
    angularSpringConstant=500
    bAngularSpring=true
    bGravityOverrideInSimSpace=true
    bOverrideAngularDamping=true
    bUseSphericalLimits=true
    simulationSpace=BoneRelative
    alphaBoolBlend {blendOption=Linear}
    boundBone {boneName=dyn_jacketFront_01_l}
    componentAppliedLinearAccClamp=<3200 3200 3200>
    componentLinearAccScale=<1.2 1.2 1.2>
    componentLinearVelScale=<1.2 1.2 1.2>
    physicsBodyDefinitions [{
      sphereCollisionRadius=0
      boundBone {boneName=dyn_jacketFront_01_l}
      boxExtents=<15 20 20>
      constraintSetup {angularLimitsMax=<1 3 10>, angularLimitsMin=<-1 -20 -10>, angularTarget=<0 0 0>}
      localJointOffset=<3 0 0>
     }]
    relativeSpaceBone {boneName=root}
    sphericalLimits [{limitRadius=2, drivingBone {boneName=thigh_l}, sphereLocalOffset=<0 0 0>}]
   }
   n20 = Anim Dynamics - Bone: dyn_jacketFront_02_l(ComponentPose=@n19, Alpha=1, ExternalForce="0,0,0") {
    angularDampingOverride=0.9999
    angularSpringConstant=500
    bAngularSpring=true
    bGravityOverrideInSimSpace=true
    bOverrideAngularDamping=true
    bUseSphericalLimits=true
    simulationSpace=BoneRelative
    alphaBoolBlend {blendOption=Linear}
    boundBone {boneName=dyn_jacketFront_02_l}
    componentAppliedLinearAccClamp=<3200 3200 3200>
    componentLinearAccScale=<1.2 1.2 1.2>
    componentLinearVelScale=<1.2 1.2 1.2>
    physicsBodyDefinitions [{
      sphereCollisionRadius=0
      boundBone {boneName=dyn_jacketFront_02_l}
      boxExtents=<15 20 20>
      constraintSetup {angularLimitsMax=<1 3 10>, angularLimitsMin=<-1 -20 -10>, angularTarget=<0 0 0>}
      localJointOffset=<3 0 0>
     }]
    relativeSpaceBone {boneName=root}
    sphericalLimits [{limitRadius=9, drivingBone {boneName=thigh_l}, sphereLocalOffset=<-9 0 -1>}]
   }
   n21 = Anim Dynamics - Bone: dyn_jacketFront_03_l(ComponentPose=@n20, Alpha=1, ExternalForce="0,0,0") {
    angularDampingOverride=0.99
    angularSpringConstant=100
    bAngularSpring=true
    bGravityOverrideInSimSpace=true
    bOverrideAngularDamping=true
    bUseSphericalLimits=true
    gravityScale=0.35
    simulationSpace=BoneRelative
    alphaBoolBlend {blendOption=Linear}
    boundBone {boneName=dyn_jacketFront_03_l}
    componentAppliedLinearAccClamp=<3000 3000 3000>
    componentLinearAccScale=<1.2 1.2 0.7>
    componentLinearVelScale=<1.2 1.2 0.7>
    physicsBodyDefinitions [{
      sphereCollisionRadius=0
      boundBone {boneName=dyn_jacketFront_03_l}
      boxExtents=<15 20 20>
      constraintSetup {angularLimitsMax=<1 3 10>, angularLimitsMin=<-1 -20 -10>, angularTarget=<0 0 0>}
      localJointOffset=<3 0 0>
     }]
    relativeSpaceBone {boneName=root}
    sphericalLimits [{limitRadius=9, drivingBone {boneName=thigh_l}, sphereLocalOffset=<-9 0 -1>}]
   }
   n22 = Anim Dynamics - Bone: dyn_jacketFront_04_l(ComponentPose=@n21, Alpha=1, ExternalForce="0,0,0") {
    angularDampingOverride=0.99
    angularSpringConstant=100
    bAngularSpring=true
    bGravityOverrideInSimSpace=true
    bOverrideAngularDamping=true
    bUseSphericalLimits=true
    gravityScale=0.35
    simulationSpace=BoneRelative
    alphaBoolBlend {blendOption=Linear}
    boundBone {boneName=dyn_jacketFront_04_l}
    componentAppliedLinearAccClamp=<3000 3000 3000>
    componentLinearAccScale=<1.2 0.7 0.7>
    componentLinearVelScale=<1.2 0.7 0.7>
    physicsBodyDefinitions [{
      sphereCollisionRadius=0
      boundBone {boneName=dyn_jacketFront_04_l}
      boxExtents=<15 20 20>
      constraintSetup {angularLimitsMax=<1 3 10>, angularLimitsMin=<-1 -20 -10>, angularTarget=<0 0 0>}
      localJointOffset=<2 0 0>
     }]
    relativeSpaceBone {boneName=root}
    sphericalLimits [{limitRadius=12, drivingBone {boneName=calf_l}, sphereLocalOffset=<-5 0 0>}]
   }
   n23 = Anim Dynamics - Bone: dyn_jacketFront_05_l(ComponentPose=@n22, Alpha=1, ExternalForce="0,0,0") {
    angularDampingOverride=0.99
    angularSpringConstant=100
    bAngularSpring=true
    bGravityOverrideInSimSpace=true
    bOverrideAngularDamping=true
    bUseSphericalLimits=true
    gravityScale=0.35
    simulationSpace=BoneRelative
    alphaBoolBlend {blendOption=Linear}
    boundBone {boneName=dyn_jacketFront_05_l}
    componentAppliedLinearAccClamp=<3000 3000 3000>
    componentLinearAccScale=<1.2 0.7 0.7>
    componentLinearVelScale=<1.2 0.7 0.7>
    physicsBodyDefinitions [{
      sphereCollisionRadius=0
      boundBone {boneName=dyn_jacketFront_05_l}
      boxExtents=<15 20 20>
      constraintSetup {angularLimitsMax=<1 3 10>, angularLimitsMin=<-1 -20 -10>, angularTarget=<0 0 0>}
      localJointOffset=<2 0 0>
     }]
    relativeSpaceBone {boneName=root}
    sphericalLimits [{limitRadius=12, drivingBone {boneName=calf_l}, sphereLocalOffset=<-5 0 0>}]
   }
   n24 = Anim Dynamics - Bone: dyn_jacketSide_02_l(ComponentPose=@n23, Alpha=1, ExternalForce="0,0,0") {
    angularDampingOverride=0.9999
    angularSpringConstant=500
    bAngularSpring=true
    bGravityOverrideInSimSpace=true
    bOverrideAngularDamping=true
    bUseSphericalLimits=true
    simulationSpace=BoneRelative
    alphaBoolBlend {blendOption=Linear}
    boundBone {boneName=dyn_jacketSide_02_l}
    componentAppliedLinearAccClamp=<3200 3200 3200>
    componentLinearAccScale=<1.2 1.2 1.2>
    componentLinearVelScale=<1.2 1.2 1.2>
    physicsBodyDefinitions [{
      sphereCollisionRadius=0
      boundBone {boneName=dyn_jacketSide_02_l}
      boxExtents=<15 20 20>
      constraintSetup {angularLimitsMax=<1 3 15>, angularLimitsMin=<-1 -20 -15>, angularTarget=<0 0 0>}
      localJointOffset=<3 0 0>
     }]
    relativeSpaceBone {boneName=root}
    sphericalLimits [{limitRadius=10, drivingBone {boneName=thigh_l}, sphereLocalOffset=<-9 0 -0.5>}]
   }
   n25 = Anim Dynamics - Bone: dyn_jacketSide_03_l(ComponentPose=@n24, Alpha=1, ExternalForce="0,0,0") {
    angularDampingOverride=0.99
    angularSpringConstant=100
    bAngularSpring=true
    bGravityOverrideInSimSpace=true
    bOverrideAngularDamping=true
    bUseSphericalLimits=true
    gravityScale=0.35
    simulationSpace=BoneRelative
    alphaBoolBlend {blendOption=Linear}
    boundBone {boneName=dyn_jacketSide_03_l}
    componentAppliedLinearAccClamp=<3000 3000 3000>
    componentLinearAccScale=<1.2 1.2 0.7>
    componentLinearVelScale=<1.2 1.2 0.7>
    physicsBodyDefinitions [{
      sphereCollisionRadius=0
      boundBone {boneName=dyn_jacketSide_03_l}
      boxExtents=<15 20 20>
      constraintSetup {angularLimitsMax=<1 3 15>, angularLimitsMin=<-1 -20 -15>, angularTarget=<0 0 0>}
      localJointOffset=<3 0 0>
     }]
    relativeSpaceBone {boneName=root}
    sphericalLimits [{limitRadius=10, drivingBone {boneName=thigh_l}, sphereLocalOffset=<-9 0 -0.5>}]
   }
   n26 = Anim Dynamics - Bone: dyn_jacketSide_04_l(ComponentPose=@n25, Alpha=1, ExternalForce="0,0,0") {
    angularDampingOverride=0.99
    angularSpringConstant=100
    bAngularSpring=true
    bGravityOverrideInSimSpace=true
    bOverrideAngularDamping=true
    bUseSphericalLimits=true
    gravityScale=0.35
    simulationSpace=BoneRelative
    alphaBoolBlend {blendOption=Linear}
    boundBone {boneName=dyn_jacketSide_04_l}
    componentAppliedLinearAccClamp=<3000 3000 3000>
    componentLinearAccScale=<1.2 0.7 0.7>
    componentLinearVelScale=<1.2 0.7 0.7>
    physicsBodyDefinitions [{
      sphereCollisionRadius=0
      boundBone {boneName=dyn_jacketSide_04_l}
      boxExtents=<15 20 20>
      constraintSetup {angularLimitsMax=<1 3 15>, angularLimitsMin=<-1 -20 -15>, angularTarget=<0 0 0>}
      localJointOffset=<2 0 0>
     }]
    relativeSpaceBone {boneName=root}
    sphericalLimits [{limitRadius=12, drivingBone {boneName=calf_l}, sphereLocalOffset=<-5 0 0>}]
   }
   n27 = Anim Dynamics - Bone: dyn_jacketSide_05_l(ComponentPose=@n26, Alpha=1, ExternalForce="0,0,0") {
    angularDampingOverride=0.99
    angularSpringConstant=100
    bAngularSpring=true
    bGravityOverrideInSimSpace=true
    bOverrideAngularDamping=true
    bUseSphericalLimits=true
    gravityScale=0.35
    simulationSpace=BoneRelative
    alphaBoolBlend {blendOption=Linear}
    boundBone {boneName=dyn_jacketSide_05_l}
    componentAppliedLinearAccClamp=<3000 3000 3000>
    componentLinearAccScale=<1.2 0.7 0.7>
    componentLinearVelScale=<1.2 0.7 0.7>
    physicsBodyDefinitions [{
      sphereCollisionRadius=0
      boundBone {boneName=dyn_jacketSide_05_l}
      boxExtents=<15 20 20>
      constraintSetup {angularLimitsMax=<1 3 15>, angularLimitsMin=<-1 -20 -15>, angularTarget=<0 0 0>}
      localJointOffset=<2 0 0>
     }]
    relativeSpaceBone {boneName=root}
    sphericalLimits [{limitRadius=12, drivingBone {boneName=calf_l}, sphereLocalOffset=<-5 0 0>}]
   }
   n28 = Anim Dynamics - Bone: dyn_jacketBack_02_l(ComponentPose=@n27, Alpha=1, ExternalForce="0,0,0") {
    angularDampingOverride=0.9999
    angularSpringConstant=500
    bAngularSpring=true
    bGravityOverrideInSimSpace=true
    bOverrideAngularDamping=true
    bUseSphericalLimits=true
    simulationSpace=BoneRelative
    alphaBoolBlend {blendOption=Linear}
    boundBone {boneName=dyn_jacketBack_02_l}
    componentAppliedLinearAccClamp=<3200 3200 3200>
    componentLinearAccScale=<1.2 1.2 1.2>
    componentLinearVelScale=<1.2 1.2 1.2>
    physicsBodyDefinitions [{
      sphereCollisionRadius=0
      boundBone {boneName=dyn_jacketBack_02_l}
      boxExtents=<15 20 20>
      constraintSetup {angularLimitsMax=<1 5 10>, angularLimitsMin=<-1 -30 -10>, angularTarget=<0 0 0>}
      localJointOffset=<5 0 0>
     }]
    relativeSpaceBone {boneName=root}
    sphericalLimits [{limitRadius=10, drivingBone {boneName=thigh_l}, sphereLocalOffset=<-9 0 -0.5>}]
   }
   n29 = Anim Dynamics - Bone: dyn_jacketBack_03_l(ComponentPose=@n28, Alpha=1, ExternalForce="0,0,0") {
    angularDampingOverride=0.99
    angularSpringConstant=100
    bAngularSpring=true
    bGravityOverrideInSimSpace=true
    bOverrideAngularDamping=true
    bUseSphericalLimits=true
    gravityScale=0.35
    simulationSpace=BoneRelative
    alphaBoolBlend {blendOption=Linear}
    boundBone {boneName=dyn_jacketBack_03_l}
    componentAppliedLinearAccClamp=<3000 3000 3000>
    componentLinearAccScale=<1.2 1.2 0.7>
    componentLinearVelScale=<1.2 1.2 0.7>
    physicsBodyDefinitions [{
      sphereCollisionRadius=0
      boundBone {boneName=dyn_jacketBack_03_l}
      boxExtents=<15 20 20>
      constraintSetup {angularLimitsMax=<1 5 10>, angularLimitsMin=<-1 -30 -10>, angularTarget=<0 0 0>}
      localJointOffset=<5 0 0>
     }]
    relativeSpaceBone {boneName=root}
    sphericalLimits [{limitRadius=10, drivingBone {boneName=thigh_l}, sphereLocalOffset=<-9 0 -0.5>}]
   }
   n30 = Anim Dynamics - Bone: dyn_jacketBack_04_l(ComponentPose=@n29, Alpha=1, ExternalForce="0,0,0") {
    angularDampingOverride=0.99
    angularSpringConstant=100
    bAngularSpring=true
    bGravityOverrideInSimSpace=true
    bOverrideAngularDamping=true
    bUseSphericalLimits=true
    gravityScale=0.35
    simulationSpace=BoneRelative
    alphaBoolBlend {blendOption=Linear}
    boundBone {boneName=dyn_jacketBack_04_l}
    componentAppliedLinearAccClamp=<3000 3000 3000>
    componentLinearAccScale=<1.2 0.7 0.7>
    componentLinearVelScale=<1.2 0.7 0.7>
    physicsBodyDefinitions [{
      sphereCollisionRadius=0
      boundBone {boneName=dyn_jacketBack_04_l}
      boxExtents=<15 20 20>
      constraintSetup {angularLimitsMax=<1 5 10>, angularLimitsMin=<-1 -30 -10>, angularTarget=<0 0 0>}
      localJointOffset=<2 0 0>
     }]
    relativeSpaceBone {boneName=root}
    sphericalLimits [{limitRadius=12, drivingBone {boneName=calf_l}, sphereLocalOffset=<-5 0 0>}]
   }
   n31 = Anim Dynamics - Bone: dyn_jacketBack_05_l(ComponentPose=@n30, Alpha=1, ExternalForce="0,0,0") {
    angularDampingOverride=0.99
    angularSpringConstant=100
    bAngularSpring=true
    bGravityOverrideInSimSpace=true
    bOverrideAngularDamping=true
    bUseSphericalLimits=true
    gravityScale=0.35
    simulationSpace=BoneRelative
    alphaBoolBlend {blendOption=Linear}
    boundBone {boneName=dyn_jacketBack_05_l}
    componentAppliedLinearAccClamp=<3000 3000 3000>
    componentLinearAccScale=<1.2 0.7 0.7>
    componentLinearVelScale=<1.2 0.7 0.7>
    physicsBodyDefinitions [{
      sphereCollisionRadius=0
      boundBone {boneName=dyn_jacketBack_05_l}
      boxExtents=<15 20 20>
      constraintSetup {angularLimitsMax=<1 5 10>, angularLimitsMin=<-1 -30 -10>, angularTarget=<0 0 0>}
      localJointOffset=<2 0 0>
     }]
    relativeSpaceBone {boneName=root}
    sphericalLimits [{limitRadius=12, drivingBone {boneName=calf_l}, sphereLocalOffset=<-5 0 0>}]
   }
   n32 = "Constraint - dyn_jacketFront_01_r (P)"(ComponentPose=@n31, Alpha=1, ConstraintWeights_0=1) {
    alphaBoolBlend {blendOption=Linear}
    alphaScaleBiasClamp {bClampResult=true, clampMax=0.4}
    boneToModify {boneName=dyn_jacketFront_01_r}
    constraintSetup [{targetBone {boneName=pelvis}}]
    constraintWeights [1]
   }
   n33 = "Constraint - dyn_jacketSide_01_r (P)"(ComponentPose=@n32, Alpha=1, ConstraintWeights_0=1) {
    alphaBoolBlend {blendOption=Linear}
    alphaScaleBiasClamp {bClampResult=true, clampMax=0.4}
    boneToModify {boneName=dyn_jacketSide_01_r}
    constraintSetup [{targetBone {boneName=pelvis}}]
    constraintWeights [1]
   }
   n34 = "Constraint - dyn_jacketBack_01_r (P)"(ComponentPose=@n33, Alpha=1, ConstraintWeights_0=1) {
    alphaBoolBlend {blendOption=Linear}
    alphaScaleBiasClamp {bClampResult=true, clampMax=0.4}
    boneToModify {boneName=dyn_jacketBack_01_r}
    constraintSetup [{targetBone {boneName=pelvis}}]
    constraintWeights [1]
   }
   n35 = Anim Dynamics - Bone: dyn_jacketFront_01_r(ComponentPose=@n34, Alpha=1, ExternalForce="0,0,0") {
    angularDampingOverride=0.9999
    angularSpringConstant=500
    bAngularSpring=true
    bGravityOverrideInSimSpace=true
    bOverrideAngularDamping=true
    bUseSphericalLimits=true
    simulationSpace=BoneRelative
    alphaBoolBlend {blendOption=Linear}
    boundBone {boneName=dyn_jacketFront_01_r}
    componentAppliedLinearAccClamp=<3200 3200 3200>
    componentLinearAccScale=<1.2 1.2 1.2>
    componentLinearVelScale=<1.2 1.2 1.2>
    physicsBodyDefinitions [{
      sphereCollisionRadius=0
      boundBone {boneName=dyn_jacketFront_01_r}
      boxExtents=<15 20 20>
      constraintSetup {angularLimitsMax=<1 3 10>, angularLimitsMin=<-1 -20 -10>, angularTarget=<0 0 0>}
      localJointOffset=<3 0 0>
     }]
    relativeSpaceBone {boneName=root}
    sphericalLimits [{limitRadius=2, drivingBone {boneName=thigh_r}, sphereLocalOffset=<0 0 0>}]
   }
   n36 = Anim Dynamics - Bone: dyn_jacketFront_02_r(ComponentPose=@n35, Alpha=1, ExternalForce="0,0,0") {
    angularDampingOverride=0.9999
    angularSpringConstant=500
    bAngularSpring=true
    bGravityOverrideInSimSpace=true
    bOverrideAngularDamping=true
    bUseSphericalLimits=true
    simulationSpace=BoneRelative
    alphaBoolBlend {blendOption=Linear}
    boundBone {boneName=dyn_jacketFront_02_r}
    componentAppliedLinearAccClamp=<3200 3200 3200>
    componentLinearAccScale=<1.2 1.2 1.2>
    componentLinearVelScale=<1.2 1.2 1.2>
    physicsBodyDefinitions [{
      sphereCollisionRadius=0
      boundBone {boneName=dyn_jacketFront_02_r}
      boxExtents=<15 20 20>
      constraintSetup {angularLimitsMax=<1 3 10>, angularLimitsMin=<-1 -20 -10>, angularTarget=<0 0 0>}
      localJointOffset=<3 0 0>
     }]
    relativeSpaceBone {boneName=root}
    sphericalLimits [{limitRadius=9, drivingBone {boneName=thigh_r}, sphereLocalOffset=<9 0 1>}]
   }
   n37 = Anim Dynamics - Bone: dyn_jacketFront_03_r(ComponentPose=@n36, Alpha=1, ExternalForce="0,0,0") {
    angularDampingOverride=0.99
    angularSpringConstant=100
    bAngularSpring=true
    bGravityOverrideInSimSpace=true
    bOverrideAngularDamping=true
    bUseSphericalLimits=true
    gravityScale=0.35
    simulationSpace=BoneRelative
    alphaBoolBlend {blendOption=Linear}
    boundBone {boneName=dyn_jacketFront_03_r}
    componentAppliedLinearAccClamp=<3000 3000 3000>
    componentLinearAccScale=<1.2 1.2 0.7>
    componentLinearVelScale=<1.2 1.2 0.7>
    physicsBodyDefinitions [{
      sphereCollisionRadius=0
      boundBone {boneName=dyn_jacketFront_03_r}
      boxExtents=<15 20 20>
      constraintSetup {angularLimitsMax=<1 3 10>, angularLimitsMin=<-1 -20 -10>, angularTarget=<0 0 0>}
      localJointOffset=<3 0 0>
     }]
    relativeSpaceBone {boneName=root}
    sphericalLimits [{limitRadius=9, drivingBone {boneName=thigh_r}, sphereLocalOffset=<9 0 1>}]
   }
   n38 = Anim Dynamics - Bone: dyn_jacketFront_04_r(ComponentPose=@n37, Alpha=1, ExternalForce="0,0,0") {
    angularDampingOverride=0.99
    angularSpringConstant=100
    bAngularSpring=true
    bGravityOverrideInSimSpace=true
    bOverrideAngularDamping=true
    bUseSphericalLimits=true
    gravityScale=0.35
    simulationSpace=BoneRelative
    alphaBoolBlend {blendOption=Linear}
    boundBone {boneName=dyn_jacketFront_04_r}
    componentAppliedLinearAccClamp=<3000 3000 3000>
    componentLinearAccScale=<1.2 0.7 0.7>
    componentLinearVelScale=<1.2 0.7 0.7>
    physicsBodyDefinitions [{
      sphereCollisionRadius=0
      boundBone {boneName=dyn_jacketFront_04_r}
      boxExtents=<15 20 20>
      constraintSetup {angularLimitsMax=<1 3 10>, angularLimitsMin=<-1 -20 -10>, angularTarget=<0 0 0>}
      localJointOffset=<2 0 0>
     }]
    relativeSpaceBone {boneName=root}
    sphericalLimits [
     {limitRadius=10, drivingBone {boneName=calf_r}, sphereLocalOffset=<-6 0 0>}
     {limitRadius=10.2, drivingBone {boneName=calf_twist_02_r}, sphereLocalOffset=<-8 0 0>}
    ]
   }
   n39 = Anim Dynamics - Bone: dyn_jacketFront_05_r(ComponentPose=@n38, Alpha=1, ExternalForce="0,0,0") {
    angularDampingOverride=0.99
    angularSpringConstant=100
    bAngularSpring=true
    bGravityOverrideInSimSpace=true
    bOverrideAngularDamping=true
    bUseSphericalLimits=true
    gravityScale=0.35
    simulationSpace=BoneRelative
    alphaBoolBlend {blendOption=Linear}
    boundBone {boneName=dyn_jacketFront_05_r}
    componentAppliedLinearAccClamp=<3000 3000 3000>
    componentLinearAccScale=<1.2 0.7 0.7>
    componentLinearVelScale=<1.2 0.7 0.7>
    physicsBodyDefinitions [{
      sphereCollisionRadius=0
      boundBone {boneName=dyn_jacketFront_05_r}
      boxExtents=<15 20 20>
      constraintSetup {angularLimitsMax=<1 3 10>, angularLimitsMin=<-1 -20 -10>, angularTarget=<0 0 0>}
      localJointOffset=<2 0 0>
     }]
    relativeSpaceBone {boneName=root}
    sphericalLimits [
     {limitRadius=10, drivingBone {boneName=calf_r}, sphereLocalOffset=<-6 0 0>}
     {limitRadius=10.2, drivingBone {boneName=calf_twist_02_r}, sphereLocalOffset=<-8 0 0>}
    ]
   }
   n40 = Anim Dynamics - Bone: dyn_jacketSide_02_r(ComponentPose=@n39, Alpha=1, ExternalForce="0,0,0") {
    angularDampingOverride=0.9999
    angularSpringConstant=500
    bAngularSpring=true
    bGravityOverrideInSimSpace=true
    bOverrideAngularDamping=true
    bUseSphericalLimits=true
    simulationSpace=BoneRelative
    alphaBoolBlend {blendOption=Linear}
    boundBone {boneName=dyn_jacketSide_02_r}
    componentAppliedLinearAccClamp=<3200 3200 3200>
    componentLinearAccScale=<1.2 1.2 1.2>
    componentLinearVelScale=<1.2 1.2 1.2>
    physicsBodyDefinitions [{
      sphereCollisionRadius=0
      boundBone {boneName=dyn_jacketSide_02_r}
      boxExtents=<15 20 20>
      constraintSetup {angularLimitsMax=<1 3 15>, angularLimitsMin=<-1 -20 -15>, angularTarget=<0 0 0>}
      localJointOffset=<3 0 0>
     }]
    relativeSpaceBone {boneName=root}
    sphericalLimits [{limitRadius=12, drivingBone {boneName=thigh_r}, sphereLocalOffset=<9 0 0.5>}]
   }
   n41 = Anim Dynamics - Bone: dyn_jacketSide_03_r(ComponentPose=@n40, Alpha=1, ExternalForce="0,0,0") {
    angularDampingOverride=0.99
    angularSpringConstant=100
    bAngularSpring=true
    bGravityOverrideInSimSpace=true
    bOverrideAngularDamping=true
    bUseSphericalLimits=true
    gravityScale=0.35
    simulationSpace=BoneRelative
    alphaBoolBlend {blendOption=Linear}
    boundBone {boneName=dyn_jacketSide_03_r}
    componentAppliedLinearAccClamp=<3000 3000 3000>
    componentLinearAccScale=<1.2 1.2 0.7>
    componentLinearVelScale=<1.2 1.2 0.7>
    physicsBodyDefinitions [{
      sphereCollisionRadius=0
      boundBone {boneName=dyn_jacketSide_03_r}
      boxExtents=<15 20 20>
      constraintSetup {angularLimitsMax=<1 3 15>, angularLimitsMin=<-1 -20 -15>, angularTarget=<0 0 0>}
      localJointOffset=<3 0 0>
     }]
    relativeSpaceBone {boneName=root}
    sphericalLimits [{limitRadius=12, drivingBone {boneName=thigh_r}, sphereLocalOffset=<9 0 0.5>}]
   }
   n42 = Anim Dynamics - Bone: dyn_jacketSide_04_r(ComponentPose=@n41, Alpha=1, ExternalForce="0,0,0") {
    angularDampingOverride=0.99
    angularSpringConstant=100
    bAngularSpring=true
    bGravityOverrideInSimSpace=true
    bOverrideAngularDamping=true
    bUseSphericalLimits=true
    gravityScale=0.35
    simulationSpace=BoneRelative
    alphaBoolBlend {blendOption=Linear}
    boundBone {boneName=dyn_jacketSide_04_r}
    componentAppliedLinearAccClamp=<3000 3000 3000>
    componentLinearAccScale=<1.2 0.7 0.7>
    componentLinearVelScale=<1.2 0.7 0.7>
    physicsBodyDefinitions [{
      sphereCollisionRadius=0
      boundBone {boneName=dyn_jacketSide_04_r}
      boxExtents=<15 20 20>
      constraintSetup {angularLimitsMax=<1 3 15>, angularLimitsMin=<-1 -20 -15>, angularTarget=<0 0 0>}
      localJointOffset=<3 0 0>
     }]
    relativeSpaceBone {boneName=root}
    sphericalLimits [
     {limitRadius=10, drivingBone {boneName=calf_r}, sphereLocalOffset=<-6 0 0>}
     {limitRadius=10.2, drivingBone {boneName=calf_twist_02_r}, sphereLocalOffset=<-8 0 0>}
    ]
   }
   n43 = Anim Dynamics - Bone: dyn_jacketSide_05_r(ComponentPose=@n42, Alpha=1, ExternalForce="0,0,0") {
    angularDampingOverride=0.99
    angularSpringConstant=100
    bAngularSpring=true
    bGravityOverrideInSimSpace=true
    bOverrideAngularDamping=true
    bUseSphericalLimits=true
    gravityScale=0.35
    simulationSpace=BoneRelative
    alphaBoolBlend {blendOption=Linear}
    boundBone {boneName=dyn_jacketSide_05_r}
    componentAppliedLinearAccClamp=<3000 3000 3000>
    componentLinearAccScale=<1.2 0.7 0.7>
    componentLinearVelScale=<1.2 0.7 0.7>
    physicsBodyDefinitions [{
      sphereCollisionRadius=0
      boundBone {boneName=dyn_jacketSide_05_r}
      boxExtents=<15 20 20>
      constraintSetup {angularLimitsMax=<1 3 15>, angularLimitsMin=<-1 -20 -15>, angularTarget=<0 0 0>}
      localJointOffset=<3 0 0>
     }]
    relativeSpaceBone {boneName=root}
    sphericalLimits [
     {limitRadius=10, drivingBone {boneName=calf_r}, sphereLocalOffset=<-6 0 0>}
     {limitRadius=10.2, drivingBone {boneName=calf_twist_02_r}, sphereLocalOffset=<-8 0 0>}
    ]
   }
   n44 = Anim Dynamics - Bone: dyn_jacketBack_02_r(ComponentPose=@n43, Alpha=1, ExternalForce="0,0,0") {
    angularDampingOverride=0.9999
    angularSpringConstant=500
    bAngularSpring=true
    bGravityOverrideInSimSpace=true
    bOverrideAngularDamping=true
    bUseSphericalLimits=true
    simulationSpace=BoneRelative
    alphaBoolBlend {blendOption=Linear}
    boundBone {boneName=dyn_jacketBack_02_r}
    componentAppliedLinearAccClamp=<3200 3200 3200>
    componentLinearAccScale=<1.2 1.2 1.2>
    componentLinearVelScale=<1.2 1.2 1.2>
    physicsBodyDefinitions [{
      sphereCollisionRadius=0
      boundBone {boneName=dyn_jacketBack_02_r}
      boxExtents=<15 20 20>
      constraintSetup {angularLimitsMax=<1 5 10>, angularLimitsMin=<-1 -30 -10>, angularTarget=<0 0 0>}
      localJointOffset=<5 0 0>
     }]
    relativeSpaceBone {boneName=root}
    sphericalLimits [{limitRadius=12, drivingBone {boneName=thigh_r}, sphereLocalOffset=<9 0 0.5>}]
   }
   n45 = Anim Dynamics - Bone: dyn_jacketBack_03_r(ComponentPose=@n44, Alpha=1, ExternalForce="0,0,0") {
    angularDampingOverride=0.99
    angularSpringConstant=100
    bAngularSpring=true
    bGravityOverrideInSimSpace=true
    bOverrideAngularDamping=true
    bUseSphericalLimits=true
    gravityScale=0.35
    simulationSpace=BoneRelative
    alphaBoolBlend {blendOption=Linear}
    boundBone {boneName=dyn_jacketBack_03_r}
    componentAppliedLinearAccClamp=<3000 3000 3000>
    componentLinearAccScale=<1.2 1.2 0.7>
    componentLinearVelScale=<1.2 1.2 0.7>
    physicsBodyDefinitions [{
      sphereCollisionRadius=0
      boundBone {boneName=dyn_jacketBack_03_r}
      boxExtents=<15 20 20>
      constraintSetup {angularLimitsMax=<1 5 10>, angularLimitsMin=<-1 -30 -10>, angularTarget=<0 0 0>}
      localJointOffset=<5 0 0>
     }]
    relativeSpaceBone {boneName=root}
    sphericalLimits [{limitRadius=12, drivingBone {boneName=thigh_r}, sphereLocalOffset=<9 0 0.5>}]
   }
   n46 = Anim Dynamics - Bone: dyn_jacketBack_04_r(ComponentPose=@n45, Alpha=1, ExternalForce="0,0,0") {
    angularDampingOverride=0.99
    angularSpringConstant=100
    bAngularSpring=true
    bGravityOverrideInSimSpace=true
    bOverrideAngularDamping=true
    bUseSphericalLimits=true
    gravityScale=0.35
    simulationSpace=BoneRelative
    alphaBoolBlend {blendOption=Linear}
    boundBone {boneName=dyn_jacketBack_04_r}
    componentAppliedLinearAccClamp=<3000 3000 3000>
    componentLinearAccScale=<1.2 0.7 0.7>
    componentLinearVelScale=<1.2 0.7 0.7>
    physicsBodyDefinitions [{
      sphereCollisionRadius=0
      boundBone {boneName=dyn_jacketBack_04_r}
      boxExtents=<15 20 20>
      constraintSetup {angularLimitsMax=<1 5 10>, angularLimitsMin=<-1 -30 -10>, angularTarget=<0 0 0>}
      localJointOffset=<2 0 0>
     }]
    relativeSpaceBone {boneName=root}
    sphericalLimits [
     {limitRadius=10, drivingBone {boneName=calf_r}, sphereLocalOffset=<-6 0 0>}
     {limitRadius=10.2, drivingBone {boneName=calf_twist_02_r}, sphereLocalOffset=<-8 0 0>}
    ]
   }
   n47 = Anim Dynamics - Bone: dyn_jacketBack_05_r(ComponentPose=@n46, Alpha=1, ExternalForce="0,0,0") {
    angularDampingOverride=0.99
    angularSpringConstant=100
    bAngularSpring=true
    bGravityOverrideInSimSpace=true
    bOverrideAngularDamping=true
    bUseSphericalLimits=true
    gravityScale=0.35
    simulationSpace=BoneRelative
    alphaBoolBlend {blendOption=Linear}
    boundBone {boneName=dyn_jacketBack_05_r}
    componentAppliedLinearAccClamp=<3000 3000 3000>
    componentLinearAccScale=<1.2 0.7 0.7>
    componentLinearVelScale=<1.2 0.7 0.7>
    physicsBodyDefinitions [{
      sphereCollisionRadius=0
      boundBone {boneName=dyn_jacketBack_05_r}
      boxExtents=<15 20 20>
      constraintSetup {angularLimitsMax=<1 5 10>, angularLimitsMin=<-1 -30 -10>, angularTarget=<0 0 0>}
      localJointOffset=<2 0 0>
     }]
    relativeSpaceBone {boneName=root}
    sphericalLimits [
     {limitRadius=10, drivingBone {boneName=calf_r}, sphereLocalOffset=<-6 0 0>}
     {limitRadius=10.2, drivingBone {boneName=calf_twist_02_r}, sphereLocalOffset=<-8 0 0>}
    ]
   }
   n48 = "Transform (Modify) Bone - Bone: dyn_belt_MSTR"!(ComponentPose=Local To Component(LocalPose=Use cached pose 'trailCache'()), Alpha=1, Rotation="-8.013,-5.289,74.827", Scale="1,1,0.2") -> (Pose) {
    rotationMode=BMM_Replace
    rotationSpace=BCS_BoneSpace
    scaleMode=BMM_Replace
    scaleSpace=BCS_BoneSpace
    translationSpace=BCS_BoneSpace
    alphaBoolBlend {blendOption=Linear}
    alphaScaleBiasClamp {bInterpResult=true, interpSpeedDecreasing=5, interpSpeedIncreasing=7}
    boneToModify {boneName=dyn_belt_MSTR}
    rotation=<rot -8.013 -5.289 74.827>
    scale {z=0.2}
   }
   n49 = Look At - Bone: dyn_belt_MSTR(ComponentPose=@n48, Alpha=1) {
    bUseLookUpAxis=true
    alphaBoolBlend {blendOption=Linear}
    boneToModify {boneName=dyn_belt_MSTR}
    lookAtLocation=<-35.4408 70.5862 -14.601>
    lookAtTarget {boneReference {boneName=pelvis}}
    lookAt_Axis {axis=<1 0>}
   }
   n50 = "Transform (Modify) Bone - Bone: dyn_jacketBack_mstr_l"!(ComponentPose=@n49, Alpha=1, Rotation="10.000,0.000,0.000") -> (Pose) {
    rotationMode=BMM_Additive
    rotationSpace=BCS_BoneSpace
    alphaBoolBlend {blendOption=Linear}
    boneToModify {boneName=dyn_jacketBack_mstr_l}
    rotation {pitch=10}
   }
   n51 = "Transform (Modify) Bone - Bone: dyn_jacketBack_mstr_r"!(ComponentPose=@n50, Alpha=1, Rotation="10,0,0") -> (Pose) {
    rotationMode=BMM_Additive
    rotationSpace=BCS_BoneSpace
    alphaBoolBlend {blendOption=Linear}
    boneToModify {boneName=dyn_jacketBack_mstr_r}
    rotation {pitch=10}
   }
   n52 = "Constraint - dyn_belt_03 (P)"(ComponentPose=@n51, Alpha=0.5, ConstraintWeights_0=1) {
    alphaBoolBlend {blendOption=Linear}
    alphaScaleBiasClamp {bInterpResult=true, clampMax=0.2, interpSpeedDecreasing=5, interpSpeedIncreasing=7}
    boneToModify {boneName=dyn_belt_03}
    constraintSetup [{targetBone {boneName=thigh_r}}]
    constraintWeights [1]
   }
   n53 = "Constraint - dyn_hipchain_l_04 (P)"(ComponentPose=@n52, Alpha=0.7, ConstraintWeights_0=1) {
    alphaBoolBlend {blendOption=Linear}
    alphaScaleBiasClamp {bClampResult=true, bInterpResult=true, clampMax=0.5, interpSpeedDecreasing=5, interpSpeedIncreasing=7}
    boneToModify {boneName=dyn_hipchain_l_04}
    constraintSetup [{targetBone {boneName=thigh_l}}]
    constraintWeights [1]
   }
   n54 = "Constraint - dyn_jacketFront_02_l (P)"(ComponentPose=@n53, Alpha=1, ConstraintWeights_0=1) {
    alphaBoolBlend {blendOption=Linear}
    alphaScaleBiasClamp {bClampResult=true, bInterpResult=true, clampMax=0.3, interpSpeedDecreasing=5, interpSpeedIncreasing=7}
    boneToModify {boneName=dyn_jacketFront_02_l}
    constraintSetup [{targetBone {boneName=thigh_l}}]
    constraintWeights [1]
   }
   n55 = "Constraint - dyn_jacketFront_02_r (P)"(ComponentPose=@n54, Alpha=1, ConstraintWeights_0=1) {
    alphaBoolBlend {blendOption=Linear}
    alphaScaleBiasClamp {bClampResult=true, bInterpResult=true, clampMax=0.3, interpSpeedDecreasing=5, interpSpeedIncreasing=7}
    boneToModify {boneName=dyn_jacketFront_02_r}
    constraintSetup [{targetBone {boneName=thigh_r}}]
    constraintWeights [1]
   }
   n56 = "Constraint - dyn_jacketSide_02_l (P)"(ComponentPose=@n55, Alpha=1, ConstraintWeights_0=1) {
    alphaBoolBlend {blendOption=Linear}
    alphaScaleBiasClamp {bClampResult=true, bInterpResult=true, clampMax=0.3, interpSpeedDecreasing=5, interpSpeedIncreasing=7}
    boneToModify {boneName=dyn_jacketSide_02_l}
    constraintSetup [{targetBone {boneName=thigh_l}}]
    constraintWeights [1]
   }
   n57 = "Constraint - dyn_jacketSide_02_r (P)"(ComponentPose=@n56, Alpha=1, ConstraintWeights_0=1) {
    alphaBoolBlend {blendOption=Linear}
    alphaScaleBiasClamp {bClampResult=true, bInterpResult=true, clampMax=0.3, interpSpeedDecreasing=5, interpSpeedIncreasing=7}
    boneToModify {boneName=dyn_jacketSide_02_r}
    constraintSetup [{targetBone {boneName=thigh_r}}]
    constraintWeights [1]
   }
   n58 = "Constraint - dyn_jacketFront_01_l (P)"(ComponentPose=@n57, Alpha=0.5, ConstraintWeights_0=1) {
    alphaBoolBlend {blendOption=Linear}
    alphaScaleBiasClamp {bInterpResult=true, clampMax=0.5, interpSpeedDecreasing=5, interpSpeedIncreasing=7}
    boneToModify {boneName=dyn_jacketFront_01_l}
    constraintSetup [{targetBone {boneName=pelvis}}]
    constraintWeights [1]
   }
   n59 = "Constraint - dyn_jacketSide_01_l (P)"(ComponentPose=@n58, Alpha=0.3, ConstraintWeights_0=1) {
    alphaBoolBlend {blendOption=Linear}
    alphaScaleBiasClamp {bInterpResult=true, clampMax=0.5, interpSpeedDecreasing=5, interpSpeedIncreasing=7}
    boneToModify {boneName=dyn_jacketSide_01_l}
    constraintSetup [{targetBone {boneName=pelvis}}]
    constraintWeights [1]
   }
   n60 = "Constraint - dyn_jacketBack_01_l (P)"(ComponentPose=@n59, Alpha=0.4, ConstraintWeights_0=1) {
    alphaBoolBlend {blendOption=Linear}
    alphaScaleBiasClamp {bInterpResult=true, clampMax=0.5, interpSpeedDecreasing=5, interpSpeedIncreasing=7}
    boneToModify {boneName=dyn_jacketBack_01_l}
    constraintSetup [{targetBone {boneName=pelvis}}]
    constraintWeights [1]
   }
   n61 = "Constraint - dyn_jacketFront_01_r (P)"(ComponentPose=@n60, Alpha=0.5, ConstraintWeights_0=1) {
    alphaBoolBlend {blendOption=Linear}
    alphaScaleBiasClamp {bInterpResult=true, clampMax=0.5, interpSpeedDecreasing=5, interpSpeedIncreasing=7}
    boneToModify {boneName=dyn_jacketFront_01_r}
    constraintSetup [{targetBone {boneName=pelvis}}]
    constraintWeights [1]
   }
   n62 = "Constraint - dyn_jacketSide_01_r (P)"(ComponentPose=@n61, Alpha=3, ConstraintWeights_0=1) {
    alphaBoolBlend {blendOption=Linear}
    alphaScaleBiasClamp {bInterpResult=true, clampMax=0.5, interpSpeedDecreasing=5, interpSpeedIncreasing=7}
    boneToModify {boneName=dyn_jacketSide_01_r}
    constraintSetup [{targetBone {boneName=pelvis}}]
    constraintWeights [1]
   }
   n63 = "Constraint - dyn_jacketBack_01_r (P)"(ComponentPose=@n62, Alpha=0.4, ConstraintWeights_0=1) {
    alphaBoolBlend {blendOption=Linear}
    alphaScaleBiasClamp {bInterpResult=true, clampMax=0.5, interpSpeedDecreasing=5, interpSpeedIncreasing=7}
    boneToModify {boneName=dyn_jacketBack_01_r}
    constraintSetup [{targetBone {boneName=pelvis}}]
    constraintWeights [1]
   }
   n64 = "Constraint - dyn_belt_03 (PPP)"!(ComponentPose=@n63, Alpha=1, ConstraintWeights_0=0.89, ConstraintWeights_1:float=0.01, ConstraintWeights_2:float=0.1) -> (Pose) {
    alphaBoolBlend {blendOption=Linear}
    boneToModify {boneName=dyn_belt_03}
    constraintSetup [{targetBone {boneName=pelvis}}, {targetBone {boneName=thigh_twist_01_r}}, {targetBone {boneName=dyn_hipchain_r}}]
    constraintWeights [1,1,1]
   }
   n65 = "Constraint - dyn_belt_04 (PP)"!(ComponentPose=@n64, Alpha=1, ConstraintWeights_0=0.4338, ConstraintWeights_1:float=0.5662) -> (Pose) {
    alphaBoolBlend {blendOption=Linear}
    boneToModify {boneName=dyn_belt_04}
    constraintSetup [{targetBone {boneName=pelvis}}, {targetBone {boneName=thigh_twist_01_r}}]
    constraintWeights [1,1]
   }
   n66 = RigidBody(ComponentPose=@n65, OverrideWorldGravity="0,0,-980") {
    bClampLinearTranslationLimitToRefPose=true
    bOverrideWorldGravity=true
    overridePhysicsAsset=&/F_MED_Silencer_FE_CoatPhysics
    simulationSpace=BaseBoneSpace
    alphaBoolBlend {blendOption=Linear}
    baseBoneRef {boneName=pelvis}
    componentAppliedLinearAccClamp=<3000 3000 3000>
    componentLinearAccScale=<0.5 0.3 0.3>
    componentLinearVelScale=<0.5 0.3 0.3>
    overrideWorldGravity {z=-980}
   }
   n67 = RigidBody(ComponentPose=@n66, OverrideWorldGravity="0,0,-980") {
    bClampLinearTranslationLimitToRefPose=true
    bOverrideWorldGravity=true
    overridePhysicsAsset=&/F_MED_Silencer_belt_Physics
    simulationSpace=BaseBoneSpace
    alphaBoolBlend {blendOption=Linear}
    baseBoneRef {boneName=pelvis}
    componentAppliedLinearAccClamp=<3000 3000 3000>
    componentLinearAccScale=<0.2 0.2 0.2>
    componentLinearVelScale=<0.3 0.3 0.3>
    overrideWorldGravity {z=-980}
    simSpaceSettings {maxAngularVelocity=3, maxLinearAcceleration=700, maxLinearVelocity=500, velocityScaleZ=0, worldAlpha=1, externalLinearDragV=<2 0.2>}
   }
   n68 = Blend Poses by bool(
    BlendPose_0=Component To Local(ComponentPose=@n67)
    BlendPose_1=Blend Poses by bool(
     BlendPose_0=Component To Local(ComponentPose=@n47)
     BlendPose_1=Component To Local(ComponentPose=@n7)
     BlendTime_0=0.3
     BlendTime_1=0.3
     bActiveValue=Get RBAN Off() -> (RBAN Off:bool)
    ) {blendType=Linear}
    BlendTime_0=0.5
    BlendTime_1=0.25
    bActiveValue=Get bIsPlayingEmote()
   ) {blendType=Linear}
   n69 = Get bClothEnabled()
   n70 = Get GravityOverride()
   n71 = RigidBody(ComponentPose=Local To Component(LocalPose=Use cached pose 'trailCache'()), OverrideWorldGravity=@n70) {
    bClampLinearTranslationLimitToRefPose=true
    bOverrideWorldGravity=true
    overridePhysicsAsset=&F_MED_Silencer_CoatPhysics
    simulationSpace=BaseBoneSpace
    alphaBoolBlend {blendOption=Linear}
    baseBoneRef {boneName=pelvis}
    componentAppliedLinearAccClamp=<3000 3000 3000>
    componentLinearAccScale=<0.5 0.4 0.4>
    componentLinearVelScale=<0.5 0.4 0.4>
   }
   n72 = RigidBody(ComponentPose=@n71, OverrideWorldGravity=@n70) {
    bClampLinearTranslationLimitToRefPose=true
    bOverrideWorldGravity=true
    overridePhysicsAsset=&F_MED_Silencer_belt_Physics
    simulationSpace=BaseBoneSpace
    alphaBoolBlend {blendOption=Linear}
    baseBoneRef {boneName=pelvis}
    componentAppliedLinearAccClamp=<3000 3000 3000>
    componentLinearAccScale=<0.2 0.2 0.2>
    componentLinearVelScale=<0.3 0.3 0.3>
    overrideWorldGravity {z=-980}
   }
   n73 = RigidBody(ComponentPose=Local To Component(LocalPose=Use cached pose 'StartupCache'()), OverrideWorldGravity="0,0,-980") {
    bClampLinearTranslationLimitToRefPose=true
    bOverrideWorldGravity=true
    overridePhysicsAsset=&F_MED_Silencer_belt_Physics
    simulationSpace=BaseBoneSpace
    alphaBoolBlend {blendOption=Linear}
    baseBoneRef {boneName=pelvis}
    componentAppliedLinearAccClamp=<3000 3000 3000>
    componentLinearAccScale=<0.2 0.2 0.2>
    componentLinearVelScale=<0.3 0.3 0.3>
    overrideWorldGravity {z=-980}
   }
   n74 = RigidBody(ComponentPose=@n73, OverrideWorldGravity="0,0,-980") {
    bClampLinearTranslationLimitToRefPose=true
    bOverrideWorldGravity=true
    overridePhysicsAsset=&/F_MED_Silencer_charm_Physics
    simulationSpace=BaseBoneSpace
    alphaBoolBlend {blendOption=Linear}
    alphaScaleBiasClamp {bInterpResult=true, interpSpeedDecreasing=5, interpSpeedIncreasing=7}
    baseBoneRef {boneName=pelvis}
    componentAppliedLinearAccClamp=<3000 3000 3000>
    componentLinearAccScale=<0.2 0.2 0.2>
    componentLinearVelScale=<0.3 0.3 0.3>
    overrideWorldGravity {z=-980}
   }
   n75 = Look At - Bone: dyn_hipchain_l_01(ComponentPose=@n74, Alpha=1) {
    alphaBoolBlend {blendOption=Linear}
    alphaScaleBiasClamp {bInterpResult=true, interpSpeedDecreasing=5, interpSpeedIncreasing=7}
    boneToModify {boneName=dyn_hipchain_l_01}
    lookAtLocation=<-21.5167 80.803 44.1239>
    lookAtTarget {boneReference {boneName=pelvis}}
    lookAt_Axis {axis=<1 0>}
   }
   n76 = Blend Poses by bool(BlendPose_0=Component To Local(ComponentPose=@n72), BlendPose_1=Component To Local(ComponentPose=@n75), bActiveValue=Get bIsDBNO()) {blendType=Linear}
   n77 = "Transform (Modify) Bone - Bone: dyn_tie_03"(
    ComponentPose=Local To Component(LocalPose=Blend Poses by bool(BlendPose_0=@n76, BlendPose_1=@n68, BlendTime_0=0.35, BlendTime_1=0.35, bActiveValue=@n69) {blendType=Linear})
    Alpha=Get AnimDyn_Off() -> (AnimDyn_Off:double)
    Translation="0,0,0"
    Rotation="0,6,0"
    Scale="1,1,1"
   ) {
    rotationMode=BMM_Additive
    rotationSpace=BCS_BoneSpace
    alphaBoolBlend {blendOption=Linear}
    boneToModify {boneName=dyn_tie_03}
    rotation {yaw=6}
   }
   n78 = "Transform (Modify) Bone - Bone: dyn_collar_r"!(ComponentPose=@n77, Alpha=1, Translation="-1,0,0") -> (Pose) {
    rotationMode=BMM_Additive
    rotationSpace=BCS_BoneSpace
    translationMode=BMM_Replace
    translationSpace=BCS_BoneSpace
    alphaBoolBlend {blendOption=Linear}
    boneToModify {boneName=dyn_collar_r}
    translation {x=-1}
   }
   Output Pose(Result=Component To Local(ComponentPose=@n78)) {layerGroup=None}
   n79 = Control Rig!(
    Source=Use cached pose 'DeformRIg'()
    Alpha
    Moving:double=Get moving() -> (moving:double)
    Jumping:double=Get Jumping() -> (Jumping:double)
    Skydive:bool=Get bIsSkyDiving()
    Slide_BaseBall:bool=Get BaseBall_Slide() -> (BaseBall_Slide:bool)
    Slide_Knee:bool=Get Knee_Slide() -> (Knee_Slide:bool)
    Crouch:bool=Get bIsCrouching()
    DBNO:bool=Get bIsDBNOCarried()
    Swing:double=Get Melee() -> (Melee:double)
    Forward:double=Get Forward() -> (Forward:double)
    Backward:double=Get Backward() -> (Backward:double)
    Left:double=Get Left() -> (Left:double)
    Right:double=Get Right() -> (Right:double)
   ) -> (Pose) {
    alphaBoolBlend {blendOption=Linear}
    controlRigAssetReference {blueprintRigClass=&/F_MED_Silencer_ClothSim_CtrlRig.F_MED_Silencer_ClothSim_CtrlRig_C}
   }
   n80 = Control Rig!(
    Source=Use cached pose 'DeformRIg'()
    Alpha
    DBNO=Get bIsDBNOCarried()
    Crouch=Get bIsCrouching()
    Skydive=Get bIsSkyDiving()
    Jumping:bool=Get Jumps() -> (Jumps:bool)
    Slide_BaseBall=Get BaseBall_Slide()
    Slide_Knee=Get Knee_Slide()
    Swing=Get Melee()
    Right=Get Right()
    Left=Get Left()
    Back:double=Get Backward()
   ) -> (Pose) {
    alphaBoolBlend {blendOption=Linear}
    controlRigAssetReference {blueprintRigClass=&/F_MED_Silencer_CtrlRig.F_MED_Silencer_CtrlRig_C}
   }
   n81 = Get bClothEnabled()
   n82 = Anim Dynamics - Bone: dyn_pendant(
    ComponentPose=Local To Component(LocalPose=Blend Poses by bool(BlendPose_0=@n79, BlendPose_1=@n80, BlendTime_0=0.5, BlendTime_1=0.5, bActiveValue=@n81) {blendType=Linear})
    Alpha=1
    ExternalForce="0,0,0"
   ) {
    angularDampingOverride=0.9999
    angularSpringConstant=500
    bAngularSpring=true
    bGravityOverrideInSimSpace=true
    bOverrideAngularDamping=true
    bUseGravityOverride=true
    bUseSphericalLimits=true
    simulationSpace=BoneRelative
    alphaBoolBlend {blendOption=Linear}
    boundBone {boneName=dyn_pendant}
    componentAppliedLinearAccClamp=<5000 5000 5000>
    componentLinearAccScale=<1.2 1.2 1.2>
    componentLinearVelScale=<1.2 1.2 1.2>
    gravityOverride {z=-980}
    physicsBodyDefinitions [{
      sphereCollisionRadius=0
      boundBone {boneName=dyn_pendant}
      boxExtents=<15 20 20>
      constraintSetup {angularLimitsMax=<1 15 10>, angularLimitsMin=<-1 -15 -0.2>, angularTarget=<0 0 0>}
      localJointOffset=<5 0 0>
     }]
    relativeSpaceBone {boneName=root}
    sphericalLimits [{limitRadius=7, drivingBone {boneName=spine_05}, sphereLocalOffset=<0 2 -5>}]
   }
   n83 = Anim Dynamics - Bone: dyn_hipchain_l_01!(ComponentPose=@n82, bAlphaBoolEnabled:bool=True, ExternalForce="0,0,0") -> (Pose) {
    alphaInputType=Bool
    angularDampingOverride=0.9999
    angularSpringConstant=500
    bAngularSpring=true
    bGravityOverrideInSimSpace=true
    bOverrideAngularDamping=true
    bUseGravityOverride=true
    bUseSphericalLimits=true
    simulationSpace=BoneRelative
    alphaBoolBlend {blendOption=Linear}
    boundBone {boneName=dyn_hipchain_l_01}
    componentAppliedLinearAccClamp=<5000 5000 5000>
    componentLinearAccScale=<1 1 1>
    componentLinearVelScale=<1 1 1>
    gravityOverride {z=-980}
    physicsBodyDefinitions [{
      sphereCollisionRadius=0
      boundBone {boneName=dyn_hipchain_l_01}
      boxExtents=<15 20 20>
      constraintSetup {angularLimitsMax=<1 5 20>, angularLimitsMin=<-1 -20 -10>, angularTarget=<0 0 0>}
      localJointOffset=<3 0 0>
     }]
    relativeSpaceBone {boneName=root}
    sphericalLimits [{limitRadius=7, drivingBone {boneName=thigh_l}, sphereLocalOffset=<0 0 0>}]
   }
   n84 = Anim Dynamics - Bone: dyn_hipchain_l_02!(ComponentPose=@n83, bAlphaBoolEnabled:bool=True, ExternalForce="0,0,0") -> (Pose) {
    alphaInputType=Bool
    angularDampingOverride=0.9999
    angularSpringConstant=500
    bAngularSpring=true
    bGravityOverrideInSimSpace=true
    bOverrideAngularDamping=true
    bUseGravityOverride=true
    bUseSphericalLimits=true
    simulationSpace=BoneRelative
    alphaBoolBlend {blendOption=Linear}
    boundBone {boneName=dyn_hipchain_l_02}
    componentAppliedLinearAccClamp=<5000 5000 5000>
    componentLinearAccScale=<1.2 1.2 1.2>
    componentLinearVelScale=<1.2 1.2 1.2>
    gravityOverride {z=-980}
    physicsBodyDefinitions [{
      sphereCollisionRadius=0
      boundBone {boneName=dyn_hipchain_l_02}
      boxExtents=<15 20 20>
      constraintSetup {angularLimitsMax=<1 8 20>, angularLimitsMin=<-1 -20 -10>, angularTarget=<0 0 0>}
      localJointOffset=<3 0 0>
     }]
    relativeSpaceBone {boneName=root}
    sphericalLimits [{limitRadius=5, drivingBone {boneName=thigh_l}, sphereLocalOffset=<0 2 0>}]
   }
   n85 = Anim Dynamics - Bone: dyn_hipchain_l_03!(ComponentPose=@n84, bAlphaBoolEnabled:bool=True, ExternalForce="0,0,0") -> (Pose) {
    alphaInputType=Bool
    angularDampingOverride=0.9999
    angularSpringConstant=500
    bAngularSpring=true
    bGravityOverrideInSimSpace=true
    bOverrideAngularDamping=true
    bUseGravityOverride=true
    bUseSphericalLimits=true
    simulationSpace=BoneRelative
    alphaBoolBlend {blendOption=Linear}
    boundBone {boneName=dyn_hipchain_l_03}
    componentAppliedLinearAccClamp=<5000 5000 5000>
    componentLinearAccScale=<1.2 1.2 1.2>
    componentLinearVelScale=<1.2 1.2 1.2>
    gravityOverride {z=-980}
    physicsBodyDefinitions [{
      sphereCollisionRadius=0
      boundBone {boneName=dyn_hipchain_l_03}
      boxExtents=<15 20 20>
      constraintSetup {angularLimitsMax=<1 8 15>, angularLimitsMin=<-1 -10 -5>, angularTarget=<0 0 0>}
      localJointOffset=<3 0 0>
     }]
    relativeSpaceBone {boneName=root}
    sphericalLimits [
     {limitRadius=7, drivingBone {boneName=thigh_l}, sphereLocalOffset=<-5 2 0>}
     {limitRadius=7, drivingBone {boneName=thigh_l}, sphereLocalOffset=<-15 0 0>}
     {limitRadius=9.5, drivingBone {boneName=pelvis}, sphereLocalOffset=<0 0 0>}
    ]
   }
   n86 = Anim Dynamics - Bone: dyn_hipchain_l_04!(ComponentPose=@n85, bAlphaBoolEnabled:bool=True, ExternalForce="0,0,0") -> (Pose) {
    alphaInputType=Bool
    angularDampingOverride=0.9999
    angularSpringConstant=500
    bAngularSpring=true
    bGravityOverrideInSimSpace=true
    bOverrideAngularDamping=true
    bUseGravityOverride=true
    bUseSphericalLimits=true
    simulationSpace=BoneRelative
    alphaBoolBlend {blendOption=Linear}
    boundBone {boneName=dyn_hipchain_l_04}
    componentAppliedLinearAccClamp=<5000 5000 5000>
    componentLinearAccScale=<1.2 1.2 1.2>
    componentLinearVelScale=<1.2 1.2 1.2>
    gravityOverride {z=-980}
    physicsBodyDefinitions [{
      sphereCollisionRadius=0
      boundBone {boneName=dyn_hipchain_l_04}
      boxExtents=<15 20 20>
      constraintSetup {angularLimitsMax=<1 8 20>, angularLimitsMin=<-1 -20 -10>, angularTarget=<0 0 0>}
      localJointOffset=<3 0 0>
     }]
    relativeSpaceBone {boneName=root}
    sphericalLimits [
     {limitRadius=7.5, drivingBone {boneName=thigh_l}, sphereLocalOffset=<-3 2 0>}
     {limitRadius=8, drivingBone {boneName=thigh_l}, sphereLocalOffset=<-12 2 0>}
    ]
   }
   n87 = Anim Dynamics - Bone: dyn_boot_top_l(ComponentPose=@n86, Alpha=1, ExternalForce="0,0,0") {
    angularDampingOverride=0.9999
    angularSpringConstant=500
    bGravityOverrideInSimSpace=true
    bOverrideAngularDamping=true
    bUseGravityOverride=true
    gravityScale=0.2
    simulationSpace=RootRelative
    alphaBoolBlend {blendOption=Linear}
    boundBone {boneName=dyn_boot_top_l}
    componentAppliedLinearAccClamp=<3500 3500 3500>
    componentLinearAccScale=<1.2 1.2 1.2>
    componentLinearVelScale=<1.2 1.2 1.2>
    gravityOverride {z=-980}
    physicsBodyDefinitions [{
      sphereCollisionRadius=0
      boundBone {boneName=dyn_boot_top_l}
      boxExtents=<5 30 30>
      constraintSetup {angularLimitsMax=<20 5 25>, angularLimitsMin=<-15 -5 -25>, angularTarget=<0 0 0>}
      localJointOffset=<3 0 0>
     }]
    relativeSpaceBone {boneName=root}
   }
   n88 = Anim Dynamics - Bone: dyn_boot_bot_l(ComponentPose=@n87, Alpha=1, ExternalForce="0,0,0") {
    angularDampingOverride=0.9999
    angularSpringConstant=500
    bGravityOverrideInSimSpace=true
    bOverrideAngularDamping=true
    bUseGravityOverride=true
    gravityScale=0.2
    simulationSpace=RootRelative
    alphaBoolBlend {blendOption=Linear}
    boundBone {boneName=dyn_boot_bot_l}
    componentAppliedLinearAccClamp=<3500 3500 3500>
    componentLinearAccScale=<1.2 1.2 1.2>
    componentLinearVelScale=<1.2 1.2 1.2>
    gravityOverride {z=-980}
    physicsBodyDefinitions [{
      sphereCollisionRadius=0
      boundBone {boneName=dyn_boot_bot_l}
      boxExtents=<5 30 30>
      constraintSetup {angularLimitsMax=<20 5 25>, angularLimitsMin=<-15 -5 -25>, angularTarget=<0 0 0>}
      localJointOffset=<3 0 0>
     }]
    relativeSpaceBone {boneName=root}
   }
   n89 = Anim Dynamics - Bone: dyn_boot_top_r(ComponentPose=@n88, Alpha=1, ExternalForce="0,0,0") {
    angularDampingOverride=0.9999
    angularSpringConstant=500
    bGravityOverrideInSimSpace=true
    bOverrideAngularDamping=true
    bUseGravityOverride=true
    gravityScale=0.2
    simulationSpace=RootRelative
    alphaBoolBlend {blendOption=Linear}
    boundBone {boneName=dyn_boot_top_r}
    componentAppliedLinearAccClamp=<3500 3500 3500>
    componentLinearAccScale=<1.2 1.2 1.2>
    componentLinearVelScale=<1.2 1.2 1.2>
    gravityOverride {z=-980}
    physicsBodyDefinitions [{
      sphereCollisionRadius=0
      boundBone {boneName=dyn_boot_top_r}
      boxExtents=<5 30 30>
      constraintSetup {angularLimitsMax=<20 5 25>, angularLimitsMin=<-15 -5 -25>, angularTarget=<0 0 0>}
      localJointOffset=<-5 0 0>
     }]
    relativeSpaceBone {boneName=root}
   }
   n90 = Anim Dynamics - Bone: dyn_boot_bot_r(ComponentPose=@n89, Alpha=1, ExternalForce="0,0,0") {
    angularDampingOverride=0.9999
    angularSpringConstant=500
    bGravityOverrideInSimSpace=true
    bOverrideAngularDamping=true
    bUseGravityOverride=true
    gravityScale=0.2
    simulationSpace=RootRelative
    alphaBoolBlend {blendOption=Linear}
    boundBone {boneName=dyn_boot_bot_r}
    componentAppliedLinearAccClamp=<3500 3500 3500>
    componentLinearAccScale=<1.2 1.2 1.2>
    componentLinearVelScale=<1.2 1.2 1.2>
    gravityOverride {z=-980}
    physicsBodyDefinitions [{
      sphereCollisionRadius=0
      boundBone {boneName=dyn_boot_bot_r}
      boxExtents=<5 30 30>
      constraintSetup {angularLimitsMax=<20 5 25>, angularLimitsMin=<-15 -5 -25>, angularTarget=<0 0 0>}
      localJointOffset=<-5 0 0>
     }]
    relativeSpaceBone {boneName=root}
   }
   n91 = Anim Dynamics - Bone: dyn_tie_01!(ComponentPose=@n90, bAlphaBoolEnabled:bool=True, GravityOverride:Vector="0,0,-980", ExternalForce="0,0,0") -> (Pose) {
    alphaInputType=Bool
    angularDampingOverride=0.99
    angularSpringConstant=100
    bAngularSpring=true
    bGravityOverrideInSimSpace=true
    bOverrideAngularDamping=true
    bUseGravityOverride=true
    bUseSphericalLimits=true
    simulationSpace=RootRelative
    alphaBoolBlend {blendOption=Linear}
    boundBone {boneName=dyn_tie_01}
    componentAppliedLinearAccClamp=<3500 3500 3500>
    componentLinearAccScale=<0.8 0.8 0.8>
    componentLinearVelScale=<0.8 0.8 0.8>
    gravityOverride {z=-980}
    physicsBodyDefinitions [{
      collisionType=CustomSphere
      sphereCollisionRadius=4.5
      boundBone {boneName=dyn_tie_01}
      boxExtents=<15 25 25>
      constraintSetup {angularLimitsMax=<0.5 20 1>, angularLimitsMin=<-0.5 -20 -1>, angularTarget=<1 0 0>}
      localJointOffset=<5 0 0>
     }]
    relativeSpaceBone {boneName=pelvis}
    sphericalLimits [
     {limitRadius=5, drivingBone {boneName=spine_05}, sphereLocalOffset=<0 0 0>}
     {limitRadius=5, drivingBone {boneName=spine_04}, sphereLocalOffset=<0 0 0>}
    ]
   }
   n92 = Anim Dynamics - Bone: dyn_tie_02!(ComponentPose=@n91, bAlphaBoolEnabled:bool=True, GravityOverride:Vector="0,0,-980", ExternalForce="0,0,0") -> (Pose) {
    alphaInputType=Bool
    angularDampingOverride=0.99
    angularSpringConstant=100
    bAngularSpring=true
    bGravityOverrideInSimSpace=true
    bOverrideAngularDamping=true
    bUseGravityOverride=true
    bUseSphericalLimits=true
    simulationSpace=RootRelative
    alphaBoolBlend {blendOption=Linear}
    boundBone {boneName=dyn_tie_02}
    componentAppliedLinearAccClamp=<3500 3500 3500>
    componentLinearAccScale=<0.8 0.8 0.8>
    componentLinearVelScale=<0.8 0.8 0.8>
    gravityOverride {z=-980}
    physicsBodyDefinitions [{
      collisionType=CustomSphere
      sphereCollisionRadius=4.5
      boundBone {boneName=dyn_tie_02}
      boxExtents=<15 25 25>
      constraintSetup {angularLimitsMax=<0.5 20 1>, angularLimitsMin=<-0.5 -20 -1>, angularTarget=<1 0 0>}
      localJointOffset=<5 0 0>
     }]
    relativeSpaceBone {boneName=pelvis}
    sphericalLimits [
     {limitRadius=5, drivingBone {boneName=spine_05}, sphereLocalOffset=<0 0 0>}
     {limitRadius=5, drivingBone {boneName=spine_04}, sphereLocalOffset=<0 0 0>}
    ]
   }
   n93 = Anim Dynamics - Bone: dyn_tie_03!(ComponentPose=@n92, bAlphaBoolEnabled:bool=True, GravityOverride:Vector="0,0,-980", ExternalForce="0,0,0") -> (Pose) {
    alphaInputType=Bool
    angularDampingOverride=0.99
    angularSpringConstant=100
    bAngularSpring=true
    bGravityOverrideInSimSpace=true
    bOverrideAngularDamping=true
    bUseGravityOverride=true
    bUseSphericalLimits=true
    simulationSpace=RootRelative
    alphaBoolBlend {blendOption=Linear}
    boundBone {boneName=dyn_tie_03}
    componentAppliedLinearAccClamp=<3500 3500 3500>
    componentLinearAccScale=<0.8 0.8 0.8>
    componentLinearVelScale=<0.8 0.8 0.8>
    gravityOverride {z=-980}
    physicsBodyDefinitions [{
      collisionType=CustomSphere
      sphereCollisionRadius=4.5
      boundBone {boneName=dyn_tie_03}
      boxExtents=<15 25 25>
      constraintSetup {angularLimitsMax=<1 10 2>, angularLimitsMin=<-1 -10 -2>, angularTarget=<1 0 0>}
      localJointOffset=<5 0 0>
     }]
    relativeSpaceBone {boneName=pelvis}
    sphericalLimits [
     {limitRadius=10, drivingBone {boneName=spine_02}, sphereLocalOffset=<0 0 0>}
     {limitRadius=8, drivingBone {boneName=spine_03}, sphereLocalOffset=<2 0 0>}
    ]
   }
   n94 = Anim Dynamics - Bone: dyn_tie_04!(ComponentPose=@n93, bAlphaBoolEnabled:bool=True, GravityOverride:Vector="0,0,-980", ExternalForce="0,0,0") -> (Pose) {
    alphaInputType=Bool
    angularDampingOverride=0.99
    angularSpringConstant=100
    bAngularSpring=true
    bGravityOverrideInSimSpace=true
    bOverrideAngularDamping=true
    bUseGravityOverride=true
    bUseSphericalLimits=true
    simulationSpace=RootRelative
    alphaBoolBlend {blendOption=Linear}
    boundBone {boneName=dyn_tie_04}
    componentAppliedLinearAccClamp=<3500 3500 3500>
    componentLinearAccScale=<0.8 0.8 0.8>
    componentLinearVelScale=<0.8 0.8 0.8>
    gravityOverride {z=-980}
    physicsBodyDefinitions [{
      collisionType=CustomSphere
      sphereCollisionRadius=4.5
      boundBone {boneName=dyn_tie_04}
      boxExtents=<15 25 25>
      constraintSetup {angularLimitsMax=<1 10 2>, angularLimitsMin=<-1 -10 -2>, angularTarget=<1 0 0>}
      localJointOffset=<5 0 0>
     }]
    relativeSpaceBone {boneName=pelvis}
    sphericalLimits [
     {limitRadius=8, drivingBone {boneName=spine_02}, sphereLocalOffset=<0 0 0>}
     {limitRadius=10, drivingBone {boneName=pelvis}, sphereLocalOffset=<0 0 0>}
     {limitRadius=8, drivingBone {boneName=thigh_l}, sphereLocalOffset=<0 0 0>}
     {limitRadius=8, drivingBone {boneName=thigh_r}, sphereLocalOffset=<0 0 0>}
    ]
   }
   n95 = Anim Dynamics - Bone: dyn_belt_03!(ComponentPose=@n94, bAlphaBoolEnabled:bool=True, GravityOverride:Vector="0,0,-980", ExternalForce="0,0,0") -> (Pose) {
    alphaInputType=Bool
    angularDampingOverride=0.99
    angularSpringConstant=100
    bAngularSpring=true
    bOverrideAngularDamping=true
    bUseGravityOverride=true
    bUseSphericalLimits=true
    simulationSpace=RootRelative
    alphaBoolBlend {blendOption=Linear}
    boundBone {boneName=dyn_belt_03}
    componentAppliedLinearAccClamp=<2000 2000 2000>
    componentLinearAccScale=<0.8 0.8 0.8>
    componentLinearVelScale=<0.8 0.8 0.8>
    gravityOverride {z=-980}
    physicsBodyDefinitions [{
      collisionType=CustomSphere
      sphereCollisionRadius=4.5
      boundBone {boneName=dyn_belt_03}
      boxExtents=<25 15 15>
      constraintSetup {angularLimitsMax=<1 20 20>, angularLimitsMin=<-1 -20 -20>, angularTarget=<1 0 0>}
      localJointOffset=<2 0 0>
     }]
    relativeSpaceBone {boneName=pelvis}
    sphericalLimits [
     {limitRadius=7, drivingBone {boneName=thigh_r}, sphereLocalOffset=<0 0 0>}
     {limitRadius=4, drivingBone {boneName=thigh_l}, sphereLocalOffset=<0 0 0>}
     {limitRadius=4, drivingBone {boneName=calf_r}, sphereLocalOffset=<0 0 0>}
    ]
   }
   n96 = Anim Dynamics - Bone: dyn_belt_04!(ComponentPose=@n95, bAlphaBoolEnabled:bool=True, GravityOverride:Vector="0,0,-980", ExternalForce="0,0,0") -> (Pose) {
    alphaInputType=Bool
    angularDampingOverride=0.99
    angularSpringConstant=100
    bAngularSpring=true
    bOverrideAngularDamping=true
    bUseGravityOverride=true
    bUseSphericalLimits=true
    simulationSpace=RootRelative
    alphaBoolBlend {blendOption=Linear}
    boundBone {boneName=dyn_belt_04}
    componentAppliedLinearAccClamp=<2000 2000 2000>
    componentLinearAccScale=<0.8 0.8 0.8>
    componentLinearVelScale=<0.8 0.8 0.8>
    gravityOverride {z=-980}
    physicsBodyDefinitions [{
      collisionType=CustomSphere
      sphereCollisionRadius=4.5
      boundBone {boneName=dyn_belt_04}
      boxExtents=<25 15 15>
      constraintSetup {angularLimitsMax=<1 20 20>, angularLimitsMin=<-1 -20 -20>, angularTarget=<1 0 0>}
      localJointOffset=<2 0 0>
     }]
    relativeSpaceBone {boneName=pelvis}
    sphericalLimits [
     {limitRadius=7, drivingBone {boneName=thigh_r}, sphereLocalOffset=<3 0 0>}
     {limitRadius=4, drivingBone {boneName=thigh_l}, sphereLocalOffset=<0 0 0>}
     {limitRadius=4, drivingBone {boneName=calf_r}, sphereLocalOffset=<0 0 0>}
    ]
   }
   n97 = Anim Dynamics - Bone: dyn_belt_05!(ComponentPose=@n96, bAlphaBoolEnabled:bool=True, GravityOverride:Vector="0,0,-980", ExternalForce="0,0,0") -> (Pose) {
    alphaInputType=Bool
    angularDampingOverride=0.99
    angularSpringConstant=100
    bAngularSpring=true
    bOverrideAngularDamping=true
    bUseGravityOverride=true
    bUseSphericalLimits=true
    simulationSpace=RootRelative
    alphaBoolBlend {blendOption=Linear}
    boundBone {boneName=dyn_belt_05}
    componentAppliedLinearAccClamp=<2000 2000 2000>
    componentLinearAccScale=<0.8 0.8 0.8>
    componentLinearVelScale=<0.8 0.8 0.8>
    gravityOverride {z=-980}
    physicsBodyDefinitions [{
      collisionType=CustomSphere
      sphereCollisionRadius=4.5
      boundBone {boneName=dyn_belt_05}
      boxExtents=<25 15 15>
      constraintSetup {angularLimitsMax=<1 20 20>, angularLimitsMin=<-1 -20 -20>, angularTarget=<1 0 0>}
      localJointOffset=<2 0 0>
     }]
    relativeSpaceBone {boneName=pelvis}
    sphericalLimits [
     {limitRadius=7, drivingBone {boneName=thigh_r}, sphereLocalOffset=<10 -1 0>}
     {limitRadius=4, drivingBone {boneName=thigh_l}, sphereLocalOffset=<0 0 0>}
     {limitRadius=4, drivingBone {boneName=calf_r}, sphereLocalOffset=<0 0 0>}
    ]
   }
   n98 = Anim Dynamics - Bone: dyn_belt_06!(ComponentPose=@n97, bAlphaBoolEnabled:bool=True, GravityOverride:Vector="0,0,-980", ExternalForce="0,0,0") -> (Pose) {
    alphaInputType=Bool
    angularDampingOverride=0.99
    angularSpringConstant=100
    bAngularSpring=true
    bOverrideAngularDamping=true
    bUseGravityOverride=true
    bUseSphericalLimits=true
    simulationSpace=RootRelative
    alphaBoolBlend {blendOption=Linear}
    boundBone {boneName=dyn_belt_06}
    componentAppliedLinearAccClamp=<2000 2000 2000>
    componentLinearAccScale=<0.8 0.8 0.8>
    componentLinearVelScale=<0.8 0.8 0.8>
    gravityOverride {z=-980}
    physicsBodyDefinitions [{
      collisionType=CustomSphere
      sphereCollisionRadius=4.5
      boundBone {boneName=dyn_belt_06}
      boxExtents=<25 15 15>
      constraintSetup {angularLimitsMax=<1 20 20>, angularLimitsMin=<-1 -20 -20>, angularTarget=<1 0 0>}
      localJointOffset=<2 0 0>
     }]
    relativeSpaceBone {boneName=pelvis}
    sphericalLimits [
     {limitRadius=4, drivingBone {boneName=calf_r}, sphereLocalOffset=<0 0 0>}
     {limitRadius=5, drivingBone {boneName=calf_l}, sphereLocalOffset=<0 0 0>}
    ]
   }
   n99 = Anim Dynamics - Bone: dyn_belt_07!(ComponentPose=@n98, bAlphaBoolEnabled:bool=True, GravityOverride:Vector="0,0,-980", ExternalForce="0,0,0") -> (Pose) {
    alphaInputType=Bool
    angularDampingOverride=0.99
    angularSpringConstant=100
    bAngularSpring=true
    bOverrideAngularDamping=true
    bUseGravityOverride=true
    bUseSphericalLimits=true
    simulationSpace=RootRelative
    alphaBoolBlend {blendOption=Linear}
    boundBone {boneName=dyn_belt_07}
    componentAppliedLinearAccClamp=<2000 2000 2000>
    componentLinearAccScale=<0.8 0.8 0.8>
    componentLinearVelScale=<0.8 0.8 0.8>
    gravityOverride {z=-980}
    physicsBodyDefinitions [{
      collisionType=CustomSphere
      sphereCollisionRadius=4.5
      boundBone {boneName=dyn_belt_07}
      boxExtents=<25 15 15>
      constraintSetup {angularLimitsMax=<1 20 20>, angularLimitsMin=<-1 -20 -20>, angularTarget=<1 0 0>}
      localJointOffset=<2 0 0>
     }]
    relativeSpaceBone {boneName=pelvis}
    sphericalLimits [
     {limitRadius=4, drivingBone {boneName=calf_r}, sphereLocalOffset=<0 0 0>}
     {limitRadius=5, drivingBone {boneName=calf_l}, sphereLocalOffset=<0 0 0>}
    ]
   }
   n100 = Anim Dynamics - Bone: dyn_belt_08!(ComponentPose=@n99, bAlphaBoolEnabled:bool=True, GravityOverride:Vector="0,0,-980", ExternalForce="0,0,0") -> (Pose) {
    alphaInputType=Bool
    angularDampingOverride=0.99
    angularSpringConstant=100
    bAngularSpring=true
    bOverrideAngularDamping=true
    bUseGravityOverride=true
    bUseSphericalLimits=true
    simulationSpace=RootRelative
    alphaBoolBlend {blendOption=Linear}
    boundBone {boneName=dyn_belt_08}
    componentAppliedLinearAccClamp=<2000 2000 2000>
    componentLinearAccScale=<0.8 0.8 0.8>
    componentLinearVelScale=<0.8 0.8 0.8>
    gravityOverride {z=-980}
    physicsBodyDefinitions [{
      collisionType=CustomSphere
      sphereCollisionRadius=4.5
      boundBone {boneName=dyn_belt_08}
      boxExtents=<25 15 15>
      constraintSetup {angularLimitsMax=<1 20 20>, angularLimitsMin=<-1 -20 -20>, angularTarget=<1 0 0>}
      localJointOffset=<2 0 0>
     }]
    relativeSpaceBone {boneName=pelvis}
    sphericalLimits [
     {limitRadius=4, drivingBone {boneName=calf_r}, sphereLocalOffset=<0 0 0>}
     {limitRadius=5, drivingBone {boneName=calf_l}, sphereLocalOffset=<0 0 0>}
    ]
   }
   n101 = Anim Dynamics - Bone: dyn_sleeve_top_l(ComponentPose=@n100, Alpha=1, ExternalForce="0,0,0") {
    angularDampingOverride=0.97
    angularSpringConstant=20000
    bAngularSpring=true
    bOverrideAngularDamping=true
    bUseGravityOverride=true
    bUseSphericalLimits=true
    simulationSpace=BoneRelative
    alphaBoolBlend {blendOption=Linear}
    boundBone {boneName=dyn_sleeve_top_l}
    componentAppliedLinearAccClamp=<2000 2000 2000>
    componentLinearAccScale=<0.1 0.1 0.2>
    componentLinearVelScale=<0.2 0.2 0.4>
    gravityOverride {z=-980}
    physicsBodyDefinitions [{
      boundBone {boneName=dyn_sleeve_top_l}
      boxExtents=<10 20 20>
      constraintSetup {angularLimitsMax=<2 5 5>, angularLimitsMin=<-2 -5 -5>, angularTarget=<0 0 0>}
      localJointOffset=<4 0 0>
     }]
    relativeSpaceBone {boneName=clavicle_l}
    sphericalLimits [{limitRadius=9, drivingBone {boneName=hand_l}, sphereLocalOffset=<0 0 0>}]
   }
   n102 = Anim Dynamics - Bone: dyn_sleeve_bot_l(ComponentPose=@n101, Alpha=1, ExternalForce="0,0,0") {
    angularDampingOverride=0.97
    angularSpringConstant=20000
    bAngularSpring=true
    bOverrideAngularDamping=true
    bUseGravityOverride=true
    bUseSphericalLimits=true
    simulationSpace=BoneRelative
    alphaBoolBlend {blendOption=Linear}
    boundBone {boneName=dyn_sleeve_bot_l}
    componentAppliedLinearAccClamp=<2000 2000 2000>
    componentLinearAccScale=<0.1 0.1 0.2>
    componentLinearVelScale=<0.2 0.2 0.4>
    gravityOverride {z=-980}
    physicsBodyDefinitions [{
      boundBone {boneName=dyn_sleeve_bot_l}
      boxExtents=<10 20 20>
      constraintSetup {angularLimitsMax=<2 5 5>, angularLimitsMin=<-2 -5 -5>, angularTarget=<0 0 0>}
      localJointOffset=<4 0 0>
     }]
    relativeSpaceBone {boneName=clavicle_l}
    sphericalLimits [{limitRadius=9, drivingBone {boneName=hand_l}, sphereLocalOffset=<0 0 0>}]
   }
   n103 = Anim Dynamics - Bone: dyn_sleeve_in_l(ComponentPose=@n102, Alpha=1, ExternalForce="0,0,0") {
    angularDampingOverride=0.97
    angularSpringConstant=20000
    bAngularSpring=true
    bOverrideAngularDamping=true
    bUseGravityOverride=true
    bUseSphericalLimits=true
    simulationSpace=BoneRelative
    alphaBoolBlend {blendOption=Linear}
    boundBone {boneName=dyn_sleeve_in_l}
    componentAppliedLinearAccClamp=<2000 2000 2000>
    componentLinearAccScale=<0.1 0.1 0.2>
    componentLinearVelScale=<0.2 0.2 0.4>
    gravityOverride {z=-980}
    physicsBodyDefinitions [{
      boundBone {boneName=dyn_sleeve_in_l}
      boxExtents=<10 20 20>
      constraintSetup {angularLimitsMax=<2 5 5>, angularLimitsMin=<-2 -5 -5>, angularTarget=<0 0 0>}
      localJointOffset=<4 0 0>
     }]
    relativeSpaceBone {boneName=clavicle_l}
    sphericalLimits [{limitRadius=9, drivingBone {boneName=hand_l}, sphereLocalOffset=<0 0 0>}]
   }
   n104 = Anim Dynamics - Bone: dyn_sleeve_out_l(ComponentPose=@n103, Alpha=1, ExternalForce="0,0,0") {
    angularDampingOverride=0.97
    angularSpringConstant=20000
    bAngularSpring=true
    bOverrideAngularDamping=true
    bUseGravityOverride=true
    bUseSphericalLimits=true
    simulationSpace=BoneRelative
    alphaBoolBlend {blendOption=Linear}
    boundBone {boneName=dyn_sleeve_out_l}
    componentAppliedLinearAccClamp=<2000 2000 2000>
    componentLinearAccScale=<0.1 0.1 0.2>
    componentLinearVelScale=<0.2 0.2 0.4>
    gravityOverride {z=-980}
    physicsBodyDefinitions [{
      boundBone {boneName=dyn_sleeve_out_l}
      boxExtents=<10 20 20>
      constraintSetup {angularLimitsMax=<2 5 5>, angularLimitsMin=<-2 -5 -5>, angularTarget=<0 0 0>}
      localJointOffset=<4 0 0>
     }]
    relativeSpaceBone {boneName=clavicle_l}
    sphericalLimits [{limitRadius=9, drivingBone {boneName=hand_l}, sphereLocalOffset=<0 0 0>}]
   }
   n105 = Anim Dynamics - Bone: dyn_sleeve_top_r(ComponentPose=@n104, Alpha=1, ExternalForce="0,0,0") {
    angularDampingOverride=0.97
    bAngularSpring=true
    bOverrideAngularDamping=true
    bUseGravityOverride=true
    bUseSphericalLimits=true
    simulationSpace=BoneRelative
    alphaBoolBlend {blendOption=Linear}
    boundBone {boneName=dyn_sleeve_top_r}
    componentAppliedLinearAccClamp=<2000 2000 2000>
    componentLinearAccScale=<0.1 0.1 0.2>
    componentLinearVelScale=<0.2 0.2 0.4>
    gravityOverride {z=-980}
    physicsBodyDefinitions [{
      boundBone {boneName=dyn_sleeve_top_r}
      boxExtents=<10 20 20>
      constraintSetup {angularLimitsMax=<2 5 5>, angularLimitsMin=<-2 -5 -5>, angularTarget=<0 0 0>}
      localJointOffset=<4 0 0>
     }]
    relativeSpaceBone {boneName=clavicle_r}
    sphericalLimits [{limitRadius=9, drivingBone {boneName=hand_r}, sphereLocalOffset=<0 0 0>}]
   }
   n106 = Anim Dynamics - Bone: dyn_sleeve_bot_r(ComponentPose=@n105, Alpha=1, ExternalForce="0,0,0") {
    angularDampingOverride=0.97
    bAngularSpring=true
    bOverrideAngularDamping=true
    bUseGravityOverride=true
    bUseSphericalLimits=true
    simulationSpace=BoneRelative
    alphaBoolBlend {blendOption=Linear}
    boundBone {boneName=dyn_sleeve_bot_r}
    componentAppliedLinearAccClamp=<2000 2000 2000>
    componentLinearAccScale=<0.1 0.1 0.2>
    componentLinearVelScale=<0.2 0.2 0.4>
    gravityOverride {z=-980}
    physicsBodyDefinitions [{
      boundBone {boneName=dyn_sleeve_bot_r}
      boxExtents=<10 20 20>
      constraintSetup {angularLimitsMax=<2 5 5>, angularLimitsMin=<-2 -5 -5>, angularTarget=<0 0 0>}
      localJointOffset=<4 0 0>
     }]
    relativeSpaceBone {boneName=clavicle_r}
    sphericalLimits [{limitRadius=9, drivingBone {boneName=hand_r}, sphereLocalOffset=<0 0 0>}]
   }
   n107 = Anim Dynamics - Bone: dyn_sleeve_out_r(ComponentPose=@n106, Alpha=1, ExternalForce="0,0,0") {
    angularDampingOverride=0.97
    bAngularSpring=true
    bOverrideAngularDamping=true
    bUseGravityOverride=true
    bUseSphericalLimits=true
    simulationSpace=BoneRelative
    alphaBoolBlend {blendOption=Linear}
    boundBone {boneName=dyn_sleeve_out_r}
    componentAppliedLinearAccClamp=<2000 2000 2000>
    componentLinearAccScale=<0.1 0.1 0.2>
    componentLinearVelScale=<0.2 0.2 0.4>
    gravityOverride {z=-980}
    physicsBodyDefinitions [{
      boundBone {boneName=dyn_sleeve_out_r}
      boxExtents=<10 20 20>
      constraintSetup {angularLimitsMax=<2 5 5>, angularLimitsMin=<-2 -5 -5>, angularTarget=<0 0 0>}
      localJointOffset=<4 0 0>
     }]
    relativeSpaceBone {boneName=clavicle_r}
    sphericalLimits [{limitRadius=9, drivingBone {boneName=hand_r}, sphereLocalOffset=<0 0 0>}]
   }
   n108 = Anim Dynamics - Bone: dyn_sleeve_in_r(ComponentPose=@n107, Alpha=1, ExternalForce="0,0,0") {
    angularDampingOverride=0.97
    bAngularSpring=true
    bOverrideAngularDamping=true
    bUseGravityOverride=true
    bUseSphericalLimits=true
    simulationSpace=BoneRelative
    alphaBoolBlend {blendOption=Linear}
    boundBone {boneName=dyn_sleeve_in_r}
    componentAppliedLinearAccClamp=<2000 2000 2000>
    componentLinearAccScale=<0.1 0.1 0.2>
    componentLinearVelScale=<0.2 0.2 0.4>
    gravityOverride {z=-980}
    physicsBodyDefinitions [{
      boundBone {boneName=dyn_sleeve_in_r}
      boxExtents=<10 20 20>
      constraintSetup {angularLimitsMax=<2 5 5>, angularLimitsMin=<-2 -5 -5>, angularTarget=<0 0 0>}
      localJointOffset=<4 0 0>
     }]
    relativeSpaceBone {boneName=clavicle_r}
    sphericalLimits [{limitRadius=9, drivingBone {boneName=hand_r}, sphereLocalOffset=<0 0 0>}]
   }
   n109 = Anim Dynamics - Bone: dyn_lapel_l(ComponentPose=@n108, Alpha=1, ExternalForce="0,0,0") {
    angularDampingOverride=1
    angularSpringConstant=20000
    bAngularSpring=true
    bOverrideAngularDamping=true
    bUseGravityOverride=true
    gravityScale=0.35
    simulationSpace=RootRelative
    alphaBoolBlend {blendOption=Linear}
    boundBone {boneName=dyn_lapel_l}
    componentLinearAccScale=<0.7 0.7 0.7>
    componentLinearVelScale=<0.7 0.7 0.7>
    gravityOverride {z=-980}
    physicsBodyDefinitions [{
      sphereCollisionRadius=0
      boundBone {boneName=dyn_lapel_l}
      boxExtents=<20 30 30>
      constraintSetup {angularLimitsMax=<0 5 10>, angularLimitsMin=<0 -2 -10>, angularTarget=<1 0 0>}
      localJointOffset=<4 0 0>
     }]
   }
   n110 = Anim Dynamics - Bone: dyn_lapel_r(ComponentPose=@n109, Alpha=1, ExternalForce="0,0,0") {
    angularDampingOverride=1
    angularSpringConstant=20000
    bAngularSpring=true
    bOverrideAngularDamping=true
    bUseGravityOverride=true
    gravityScale=0.35
    simulationSpace=RootRelative
    alphaBoolBlend {blendOption=Linear}
    boundBone {boneName=dyn_lapel_r}
    componentLinearAccScale=<0.7 0.7 0.7>
    componentLinearVelScale=<0.7 0.7 0.7>
    gravityOverride {z=-980}
    physicsBodyDefinitions [{
      sphereCollisionRadius=0
      boundBone {boneName=dyn_lapel_r}
      boxExtents=<20 30 30>
      constraintSetup {angularLimitsMax=<0 5 10>, angularLimitsMin=<0 -2 -10>, angularTarget=<1 0 0>}
      localJointOffset=<4 0 0>
     }]
   }
   n111 = Anim Dynamics - Bone: dyn_coat_collar_l(ComponentPose=@n110, Alpha=1, ExternalForce="0,0,0") {
    angularDampingOverride=1
    angularSpringConstant=20000
    bAngularSpring=true
    bOverrideAngularDamping=true
    bUseGravityOverride=true
    gravityScale=0.35
    simulationSpace=RootRelative
    alphaBoolBlend {blendOption=Linear}
    boundBone {boneName=dyn_coat_collar_l}
    componentLinearAccScale=<0.7 0.7 0.7>
    componentLinearVelScale=<0.7 0.7 0.7>
    gravityOverride {z=-980}
    physicsBodyDefinitions [{
      sphereCollisionRadius=0
      boundBone {boneName=dyn_coat_collar_l}
      boxExtents=<20 30 30>
      constraintSetup {angularLimitsMax=<0 5 8>, angularLimitsMin=<0 -5 -8>, angularTarget=<1 0 0>}
      localJointOffset=<4 0 0>
     }]
   }
   n112 = Anim Dynamics - Bone: dyn_coat_collar_r(ComponentPose=@n111, Alpha=1, ExternalForce="0,0,0") {
    angularDampingOverride=1
    angularSpringConstant=20000
    bAngularSpring=true
    bOverrideAngularDamping=true
    bUseGravityOverride=true
    gravityScale=0.35
    simulationSpace=RootRelative
    alphaBoolBlend {blendOption=Linear}
    boundBone {boneName=dyn_coat_collar_r}
    componentLinearAccScale=<0.7 0.7 0.7>
    componentLinearVelScale=<0.7 0.7 0.7>
    gravityOverride {z=-980}
    physicsBodyDefinitions [{
      sphereCollisionRadius=0
      boundBone {boneName=dyn_coat_collar_r}
      boxExtents=<20 30 30>
      constraintSetup {angularLimitsMax=<0 5 8>, angularLimitsMin=<0 -5 -8>, angularTarget=<1 0 0>}
      localJointOffset=<4 0 0>
     }]
   }
   n113 = Look At - Bone: dyn_sleeve_top_l(ComponentPose=@n112, Alpha=1) {
    alphaBoolBlend {blendOption=Linear}
    alphaScaleBiasClamp {bInterpResult=true, interpSpeedDecreasing=5, interpSpeedIncreasing=7}
    boneToModify {boneName=dyn_sleeve_top_l}
    lookAtLocation {x=50, z=20}
    lookAtTarget {boneReference {boneName=lowerarm_l}}
    lookAt_Axis {axis=<1 0>}
   }
   n114 = Look At - Bone: dyn_sleeve_in_r(ComponentPose=@n113, Alpha=1) {
    alphaBoolBlend {blendOption=Linear}
    alphaScaleBiasClamp {bInterpResult=true, interpSpeedDecreasing=5, interpSpeedIncreasing=7}
    boneToModify {boneName=dyn_sleeve_in_r}
    lookAtLocation=<-50 -20>
    lookAtTarget {boneReference {boneName=lowerarm_r}}
    lookAt_Axis {axis=<1 0>}
   }
   n115 = Look At - Bone: dyn_sleeve_top_l(ComponentPose=@n114, Alpha=1) {
    alphaBoolBlend {blendOption=Linear}
    alphaScaleBiasClamp {bInterpResult=true, interpSpeedDecreasing=5, interpSpeedIncreasing=7}
    boneToModify {boneName=dyn_sleeve_top_l}
    lookAtLocation=<57.4106 0 20>
    lookAtTarget {boneReference {boneName=lowerarm_l}}
    lookAt_Axis {axis=<1 0>}
   }
   n116 = Look At - Bone: dyn_sleeve_out_l(ComponentPose=@n115, Alpha=1) {
    alphaBoolBlend {blendOption=Linear}
    alphaScaleBiasClamp {bInterpResult=true, interpSpeedDecreasing=5, interpSpeedIncreasing=7}
    boneToModify {boneName=dyn_sleeve_out_l}
    lookAtLocation=<50 -16.8781 10.1012>
    lookAtTarget {boneReference {boneName=lowerarm_l}}
    lookAt_Axis {axis=<1 0>}
   }
   n117 = Look At - Bone: dyn_sleeve_top_r(ComponentPose=@n116, Alpha=1) {
    alphaBoolBlend {blendOption=Linear}
    alphaScaleBiasClamp {bInterpResult=true, interpSpeedDecreasing=5, interpSpeedIncreasing=7}
    boneToModify {boneName=dyn_sleeve_top_r}
    lookAtLocation=<-45.321 6.9556 -9.455>
    lookAtTarget {boneReference {boneName=lowerarm_r}}
    lookAt_Axis {axis=<1 0>}
   }
   n118 = Look At - Bone: dyn_sleeve_in_r(ComponentPose=@n117, Alpha=1) {
    alphaBoolBlend {blendOption=Linear}
    alphaScaleBiasClamp {bInterpResult=true, interpSpeedDecreasing=5, interpSpeedIncreasing=7}
    boneToModify {boneName=dyn_sleeve_in_r}
    lookAtLocation=<-50 -24.1851 0>
    lookAtTarget {boneReference {boneName=lowerarm_r}}
    lookAt_Axis {axis=<1 0>}
   }
   n119 = Look At - Bone: dyn_sleeve_out_r(ComponentPose=@n118, Alpha=1) {
    alphaBoolBlend {blendOption=Linear}
    alphaScaleBiasClamp {bInterpResult=true, interpSpeedDecreasing=5, interpSpeedIncreasing=7}
    boneToModify {boneName=dyn_sleeve_out_r}
    lookAtLocation=<-50 30.7026 0>
    lookAtTarget {boneReference {boneName=lowerarm_r}}
    lookAt_Axis {axis=<1 0>}
   }
   n120 = "Transform (Modify) Bone - Bone: dyn_belt_MSTR"!(ComponentPose=@n119, Alpha=1, Rotation="-0.000,-0.000,60.000") -> (Pose) {
    rotationMode=BMM_Replace
    rotationSpace=BCS_BoneSpace
    alphaBoolBlend {blendOption=Linear}
    boneToModify {boneName=dyn_belt_MSTR}
    rotation=<rot 0 0 60>
   }
   n121 = "Transform (Modify) Bone - Bone: floor"!(ComponentPose=@n120, Alpha=1, Translation="0,0,0", Rotation=Get floor_slope() -> (floor_slope:Rotator)) -> (Pose) {
    rotationMode=BMM_Replace
    rotationSpace=BCS_WorldSpace
    translationSpace=BCS_WorldSpace
    alphaBoolBlend {blendOption=Linear}
    boneToModify {boneName=floor}
    rotation {pitch=0, roll=60}
   }
   Save cached pose 'StartupCache'(Pose=Component To Local(ComponentPose=@n121)) {cacheName=StartupCache}
   n122 = Get BwdMovement() -> (BwdMovement:double)
   n123 = Trail controller - Bone: dyn_jacketFront_05_l(ComponentPose=Local To Component(LocalPose=Use cached pose 'StartupCache'()), Alpha=@n122) {
    bUsePlanarLimit=true
    chainLength=5
    debugLifeTime=0.1
    maxDeltaTime=0.3
    relaxationSpeedScale=1.5
    alphaScaleBiasClamp {bInterpResult=true}
    rotationLimits [{limitMax=<180 180 180>}, {limitMax=<180 180 180>}, {limitMax=<180 180 180>}, {limitMax=<180 180 180>}, {limitMax=<180 180 180>}]
    rotationOffsets [<0 0 0>, <0 0 0>, <0 0 0>, <0 0 0>, <0 0 0>]
    trailBone {boneName=dyn_jacketFront_05_l}
    trailRelaxationSpeed {editorCurveData {keys [<at -0.0134 19.3841 RCIM_Linear>,<at 0.5079 12.8986 RCIM_Linear>,<at 0.9955 9.6044 RCIM_Linear>]}}
   }
   n124 = Trail controller - Bone: dyn_jacketSide_05_l(ComponentPose=@n123, Alpha=@n122) {
    bUsePlanarLimit=true
    chainLength=5
    debugLifeTime=0.1
    maxDeltaTime=0.3
    relaxationSpeedScale=1.5
    alphaScaleBiasClamp {bInterpResult=true}
    rotationLimits [{limitMax=<180 180 180>}, {limitMax=<180 180 180>}, {limitMax=<180 180 180>}, {limitMax=<180 180 180>}, {limitMax=<180 180 180>}]
    rotationOffsets [<0 0 0>, <0 0 0>, <0 0 0>, <0 0 0>, <0 0 0>]
    trailBone {boneName=dyn_jacketSide_05_l}
    trailRelaxationSpeed {editorCurveData {keys [<at -0.0134 19.3841 RCIM_Linear>,<at 0.5079 12.8986 RCIM_Linear>,<at 0.9955 9.6044 RCIM_Linear>]}}
   }
   n125 = Trail controller - Bone: dyn_jacketBack_05_l(ComponentPose=@n124, Alpha=@n122) {
    bUsePlanarLimit=true
    chainLength=5
    debugLifeTime=0.1
    maxDeltaTime=0.3
    relaxationSpeedScale=1.5
    alphaScaleBiasClamp {bInterpResult=true}
    rotationLimits [{limitMax=<180 180 180>}, {limitMax=<180 180 180>}, {limitMax=<180 180 180>}, {limitMax=<180 180 180>}, {limitMax=<180 180 180>}]
    rotationOffsets [<0 0 0>, <0 0 0>, <0 0 0>, <0 0 0>, <0 0 0>]
    trailBone {boneName=dyn_jacketBack_05_l}
    trailRelaxationSpeed {editorCurveData {keys [<at 0 15 RCIM_Linear>,<at 0.5289 13.1838 RCIM_Linear>,<at 1 6 RCIM_Linear>]}}
   }
   n126 = Trail controller - Bone: dyn_jacketFront_05_r(ComponentPose=@n125, Alpha=@n122) {
    bUsePlanarLimit=true
    chainLength=5
    debugLifeTime=0.1
    maxDeltaTime=0.3
    relaxationSpeedScale=1.5
    alphaScaleBiasClamp {bInterpResult=true}
    rotationLimits [{limitMax=<180 180 180>}, {limitMax=<180 180 180>}, {limitMax=<180 180 180>}, {limitMax=<180 180 180>}, {limitMax=<180 180 180>}]
    rotationOffsets [<0 0 0>, <0 0 0>, <0 0 0>, <0 0 0>, <0 0 0>]
    trailBone {boneName=dyn_jacketFront_05_r}
    trailRelaxationSpeed {editorCurveData {keys [<at -0.0134 19.3841 RCIM_Linear>,<at 0.5079 12.8986 RCIM_Linear>,<at 0.9955 9.6044 RCIM_Linear>]}}
   }
   n127 = Trail controller - Bone: dyn_jacketSide_05_r(ComponentPose=@n126, Alpha=@n122) {
    bUsePlanarLimit=true
    chainLength=5
    debugLifeTime=0.1
    maxDeltaTime=0.3
    relaxationSpeedScale=1.5
    alphaScaleBiasClamp {bInterpResult=true}
    rotationLimits [{limitMax=<180 180 180>}, {limitMax=<180 180 180>}, {limitMax=<180 180 180>}, {limitMax=<180 180 180>}, {limitMax=<180 180 180>}]
    rotationOffsets [<0 0 0>, <0 0 0>, <0 0 0>, <0 0 0>, <0 0 0>]
    trailBone {boneName=dyn_jacketSide_05_r}
    trailRelaxationSpeed {editorCurveData {keys [<at -0.0134 19.3841 RCIM_Linear>,<at 0.5079 12.8986 RCIM_Linear>,<at 0.9955 9.6044 RCIM_Linear>]}}
   }
   n128 = Trail controller - Bone: dyn_jacketBack_05_r(ComponentPose=@n127, Alpha=@n122) {
    bUsePlanarLimit=true
    chainLength=5
    debugLifeTime=0.1
    maxDeltaTime=0.3
    relaxationSpeedScale=1.5
    alphaScaleBiasClamp {bInterpResult=true}
    rotationLimits [
     {limitMax=<180 180 180>}
     {limitMax=<180 180 180>}
     {limitMax=<180 90 180>, limitMin=<-180 -90 -180>}
     {limitMax=<180 180 180>}
     {limitMax=<180 180 180>}
    ]
    rotationOffsets [<0 0 0>, <0 0 0>, <0 0 0>, <0 0 0>, <0 0 0>]
    trailBone {boneName=dyn_jacketBack_05_r}
    trailRelaxationSpeed {editorCurveData {keys [<at 0 15 RCIM_Linear>,<at 0.5289 13.1838 RCIM_Linear>,<at 1 6 RCIM_Linear>]}}
   }
   Save cached pose 'trailCache'(Pose=Component To Local(ComponentPose=@n128)) {cacheName=trailCache}
   Save cached pose 'DeformRIg'(Pose=Copy Pose From Mesh() {bUseAttachedParent=true}) {cacheName=DeformRIg}
   Use cached pose 'trailCache'()
  ]}
 EventGraph {nodes [
   n1 = Event BlueprintUpdateAnimation()
   > Set DeltaTimeLocal(DeltaTimeLocal:double=@n1.DeltaTimeX) -> (Output_Get:double)
   > n2 = Sequence!(execute) -> (then_0, then_1, then_2:Exec, then_3:Exec, then_4:Exec, then_5:Exec, then_6:Exec, then_7:Exec, then_8:Exec, then_9:Exec, then_10:Exec)
   n3 = OR Boolean(A=Get bIsJumping(), B=Get bIsFalling())
   Set Jumping(execute=@n2.then_1, Jumping:double=SelectFloat(A=1, B=0, bPickA=@n3)) -> (Output_Get:double)
   > Set Jumps(Jumps:bool=@n3) -> (Output_Get:bool)
   > Set Cloth Off(Cloth Off:bool=NOT Boolean(A=Get bClothEnabled())) -> (Output_Get:bool)
   Reroute Node(InputPin=@n2.then_10)
   Set Crouch(execute=@n2.then_2, Crouch:double=SelectFloat(A=1, B=0, bPickA=Get bIsCrouching())) -> (Output_Get:double)
   n4 = NOT Boolean(A=Get bIsTargeting())
   n5 = NOT Boolean(A=Get bIsSkyDiving())
   n6 = NOT Boolean(A=Get bIsJumping())
   n7 = NOT Boolean(A=Get bIsFalling())
   n8 = Get bIsMoving2D()
   > Set moving(moving:double=SelectFloat(A=1, B=0, bPickA=AND Boolean!(A=@n8, B=@n5, C:bool=@n4, D:bool=@n6, E:bool=@n7) -> (ReturnValue))) -> (Output_Get:double)
   > Set Melee(Melee:double=SelectFloat(A=1, B=0, bPickA=Get bIsPlayingMeleeAnim())) -> (Output_Get:double)
   > Set Crouch Off(Crouch Off:bool=NOT Boolean(A=Get bIsCrouching())) -> (Output_Get:bool)
   n9 = Get bAnimDynamicsEnabled() -> (bAnimDynamicsEnabled:bool)
   > Set AnimDyn_Off(AnimDyn_Off:double=SelectFloat(A=1, B=0, bPickA=NOT Boolean(A=@n9))) -> (Output_Get:double)
   > Set RBAN Off(RBAN Off:bool=NOT Boolean(A=Get bRigidBodyEnabled())) -> (Output_Get:bool)
   > Set floor_slope(floor_slope:Rotator=Break Fort Anim Input Slope Info!(FortAnimInput_SlopeInfo:"Fort Anim Input Slope Info Structure (by ref)"=Get SlopeInfoInput() -> (SlopeInfoInput:Fort Anim Input Slope Info Structure)) -> (RootSlopeTranslation:Vector=~, LeftFootSlopeTranslation:Vector=~, RightFootSlopeTranslation:Vector=~, SlopeRotation:Rotator=~) {is=title K2Node_BreakStruct}.SlopeRotation) -> (Output_Get:Rotator)
   > Set CustomAnim_Off(CustomAnim_Off:bool=NOT Boolean(A=Get bCanPlayCustomAnimations() -> (bCanPlayCustomAnimations:bool))) -> (Output_Get:bool)
   n10 = GetAnimInstance(self=GetOwnerSkeletalMeshComponent())
   > Is Valid(InputObject=@n10)
   n11 = Montage_IsActive!(
    self:Anim Instance Object Reference=@n10
    Montage:Anim Montage Object Reference=/BRCosmetics/Animation/Game/MainPlayer/Emotes/Donut1/Donut_One_CMF_M_Frontend
   ) -> (ReturnValue:bool)
   n12 = Montage_IsActive(self=@n10, Montage=/BRCosmetics/Animation/Game/MainPlayer/Emotes/Jupiter/Emote_Jupiter_CMF_M)
   > Set emote_alpha(emote_alpha:double=SelectFloat(A=1, B=0, bPickA=AND Boolean!(A=@n12, B=@n11) -> (ReturnValue))) -> (Output_Get:double)
   n13 = Branch(execute=@n2.then_3, Condition=Get bIsSliding())
   > n14 = Set Left_Calf(Left_Calf:double=GetAlphaFromReferenceTransformDelta(SkeletalMeshComponent=GetPartSkeletalMeshComponent(), BoneName=calf_l, BaseName=thigh_l, InputRange=(X=0,Y=1), TwistAxis="1,0,0").RawOutput) -> (Output_Get:double)
   > n15 = Branch(Condition=NOT Boolean(A="float > float"(A=@n14, B=100)))
   > Set BaseBall_Slide(BaseBall_Slide:bool=true) -> (Output_Get:bool)
   > Set Knee_Slide(Knee_Slide:bool=false) -> (Output_Get:bool)
   Set Knee_Slide(execute=@n15.else, Knee_Slide=true)
   > Set BaseBall_Slide(BaseBall_Slide=false)
   Set BaseBall_Slide(execute=@n13.else, BaseBall_Slide=false)
   > Set Knee_Slide(Knee_Slide=false)
   n16 = PlayerMovement(execute=@n2.then_4, AnimInstance=Self-Reference(), XDotProductCoefficient=1, YDotProductCoefficient=1, ZDotProductCoefficient=1)
   n17 = MapRangeClamped(Value=@n16.YMovement, InRangeA=0, InRangeB=1)
   n18 = "Absolute (Float)"(A=@n16)
   > Set BwdMovement(BwdMovement:double=MapRangeClamped(Value=float + float(A=@n18, B=@n17) {is=title K2Node_CommutativeAssociativeBinaryOperator}, InRangeA=0, InRangeB=1.4)) -> (Output_Get:double)
   n19 = Branch(execute=@n2.then_5, Condition=Get bIsSwimSprinting() -> (bIsSwimSprinting:bool))
   n20 = Branch(execute=@n19.else, Condition=Get bIsSkyDiving())
   > n21 = Branch(Condition=Get bIsParachuteOpened())
   n22 = Branch(execute=@n21.else, Condition="float > float"(A=Get PawnVelocityZ(), B=50) {is=title K2Node_CallFunction})
   > Set BwdMovement(BwdMovement=float * float(A=Get BwdMovement(), B=0))
   Branch(execute=@n22.else, Condition=Get bIsSkydiveDiveMode())
   Set BwdMovement(execute=@n20.else, BwdMovement=float * float(A=Get BwdMovement(), B=1))
   n23 = GetPartSkeletalMeshComponent()
   Set TeleportDistanceThreshold!(execute=@n2.then_7, TeleportDistanceThreshold:float=300, self:Skeletal Mesh Component Object Reference=@n23) -> (then, Output_Get:float)
   > Set TeleportRotationThreshold!(execute, TeleportRotationThreshold:float=15, self:Skeletal Mesh Component Object Reference=@n23) -> (then, Output_Get:float)
   n24 = PlayerMovement(execute=@n2.then_8, AnimInstance=Self-Reference(), XDotProductCoefficient=100, YDotProductCoefficient=100, ZDotProductCoefficient=100)
   n25 = BreakVector(InVec=MakeVector(X=@n24, Y=@n24.YMovement))
   > Set Left(Left:double="To Float (Boolean)"(InBool="float < float"(A=@n25, B=-50))) -> (Output_Get:double)
   > Set Right(Right:double="To Float (Boolean)"(InBool="float > float"(A=@n25, B=50) {is=title K2Node_PromotableOperator})) -> (Output_Get:double)
   > Set Forward(Forward:double="To Float (Boolean)"(InBool="float > float"(A=@n25.Y, B=40))) -> (Output_Get:double)
   > Set Backward(Backward:double="To Float (Boolean)"(InBool="float < float"(A=@n25.Y, B=-40))) -> (Output_Get:double)
   n26 = Is Valid(exec=@n2.then_9, InputObject=Self-Reference())
   n27 = CheckForBackbling(LongBackblingThreshold=40)
   n28 = Branch(execute=[@n26.Is Valid, @n49.Completed], Condition=@n27.bIsBackblingACape)
   > Set Is Backbling A Cape(Is Backbling A Cape:bool=true) -> (Output_Get:bool)
   > Set Cape Alpha(Cape Alpha:double=0) -> (Output_Get:double)
   n29 = GetPartSkeletalMeshComponent()
   > Set ClothBlendWeight!(execute, ClothBlendWeight:float=0, self:Skeletal Mesh Component Object Reference=@n29) -> (then, Output_Get:float)
   > Do Once(Reset=@n31.then)
   > n30 = SuspendClothingSimulation(self:Skeletal Mesh Component Object Reference=@n29)
   Do Once(execute=@n32.then, Reset=@n30.then)
   > n31 = ResumeClothingSimulation(self:Skeletal Mesh Component Object Reference=@n29)
   Set Is Backbling A Cape(execute=@n28.else, Is Backbling A Cape=false)
   > Set Cape Alpha(Cape Alpha=1)
   > n32 = Set ClothBlendWeight(ClothBlendWeight=1, self=@n29)
   n33 = GetClothingSimulationInteractor!(self:Skeletal Mesh Component Object Reference=GetOwningComponent(), ClothingAssetName:Name=None) -> (ReturnValue:Clothing Simulation Interactor Object Reference)
   Is Valid(exec=@n2.then_6, InputObject=@n33)
   n34 = MakeLiteralString(Value=Silencer_Cloth)
   > n35 = Cast To ChaosClothingInteractor!(
    execute
    Object=GetClothingInteractor!(self:Clothing Simulation Interactor Object Reference=@n33, ClothingAssetName:Name=String To Name(InString=@n34)) -> (ReturnValue:Clothing Interactor Object Reference)
   ) -> (then, CastFailed, AsChaos Clothing Interactor:Chaos Clothing Interactor Object Reference) {is=title K2Node_DynamicCast}
   > Set ChaosClothingInteractor(ChaosClothingInteractor:Chaos Clothing Interactor Object Reference=@n35) -> (Output_Get:Chaos Clothing Interactor Object Reference)
   n36 = Sequence!(execute=@n2.then_0) -> (then_0, then_1)
   > Branch(Condition=Get bClothEnabled())
   > n37 = Cloth Controller Function DEPRECATED!(
    execute
    CharacterPartAnimInstance:Custom Character Part Anim Instance Object Reference=Self-Reference()
    ClothingAssetName:String=Silencer_Cloth
    Cloth Paramaters:Cloth Params Struct Structure=Get Cloth Paramaters_0() -> (Cloth Paramaters_0:Cloth Params Struct Structure)
    Cloth Core Settings:Cloth Core Setting Struct Structure=Get Cloth Core Settings_0() -> (Cloth Core Settings_0:Cloth Core Setting Struct Structure)
    ClothUpAxis:Vector="1,0,0"
    OnGround_Moving_SpeedLimit:double=2
    OnGround_Moving_AngularVel_Max:double=0.05
    OnGround_Moving_FictitiousVel_Max:double=0.6
    LinVelScale_FwdBwd_Clamp:double=0
    LinVelScale_UpDwn_Clamp:double=0.7
    DeltaTime:double=Get DeltaTimeLocal() -> (DeltaTimeLocal:double)
    GameTimeInSeconds:double=GetGameTimeInSeconds()
    WindGustsParams:Wind Gust Params Struct Structure=Get Wind Gust Params() -> (Wind Gust Params:Wind Gust Params Struct Structure)
    WindGustLocalSettings:Wind Gust Local Setting Struct Structure=Get Gust Local Settings() -> (Gust Local Settings:Wind Gust Local Setting Struct Structure)
   ) -> (
    then
    CurrentStateOut:String
    Cloth Settings Out:Cloth Setting Struct Structure
    Cloth Core Settings Out:Cloth Core Setting Struct Structure
    WindGustSettings Out:Wind Gust Setting Struct Structure
    WindGustLocalSettings Out:Wind Gust Local Setting Struct Structure
    GustStrength - WindVelocityOut:Vector
    LinearVelVector-DELETE LATER:Vector
   )
   > Set Gust Local Settings(Gust Local Settings:Wind Gust Local Setting Struct Structure=@n37.WindGustLocalSettings Out) -> (Output_Get:Wind Gust Local Setting Struct Structure)
   Branch(execute=@n36.then_1, Condition=Get bRigidBodyEnabled())
   > n38 = "Gravity Override State Function (Parameter Asset) (Nativized)"(
    CharacterPartAnimInstance=Self-Reference()
    GameTimeInSeconds=GetGameTimeInSeconds()
    Parameters=&/M_MED_Silencer_Parameters
   )
   > Set GravityOverride(GravityOverride=@n38)
   n39 = Event BlueprintProcessAnimInputEvent()
   > n40 = Branch(Condition=IsNamed(self=@n39.AnimInputEvent, InEventName=aie.Player.Zipline.Entered))
   > Delay(Duration=0.1)
   n41 = Get ChaosClothingInteractor() -> (ChaosClothingInteractor:Chaos Clothing Interactor Object Reference)
   > ResetAndTeleport!(execute, self:Chaos Clothing Interactor Object Reference=@n41, bReset:bool=true, bTeleport:bool=true) -> (then)
   Branch(execute=@n40.else, Condition=IsNamed(self=@n39.AnimInputEvent, InEventName=aie.Player.Zipline.Exited))
   > Set resetClothLandingFromZilpine(resetClothLandingFromZilpine:bool=true) -> (Output_Get:bool)
   Event OnEndFalling()
   > Branch(Condition=Get resetClothLandingFromZilpine() -> (resetClothLandingFromZilpine:bool))
   > ResetAndTeleport(self=@n41, bReset=true, bTeleport=true)
   > Set resetClothLandingFromZilpine(resetClothLandingFromZilpine=false)
   Event BlueprintInitializeAnimation()
   > Set bResetAfterTeleport!(execute, bResetAfterTeleport:bool=false, self:Skeletal Mesh Component Object Reference=@n23) -> (then, Output_Get:bool)
   Event OnParachuteOpened()
   > ResetAndTeleport(self=Get ChaosClothingInteractor(), bReset=true, bTeleport=true)
   n42 = Branch()
   > Do Once(Reset=@n45.then)
   n43 = GetPartSkeletalMeshComponent()
   > n44 = SuspendClothingSimulation(self=@n43)
   Do Once(execute=@n42.else, Reset=@n44.then)
   > n45 = ResumeClothingSimulation(self=@n43)
   n46 = Branch(Condition=@n27)
   > n47 = Set IsBackblingOn(IsBackblingOn:bool=true) -> (Output_Get:bool)
   n48 = Set IsBackblingOn(execute=@n46.else, IsBackblingOn=false)
   n49 = Do Once()
   OR Boolean(A=@n27.bIsBackblingACape, B=Get bIsInFrontEnd())
   Set IsBackblingOnAlpha(IsBackblingOnAlpha:double="To Float (Boolean)"(InBool=@n47)) -> (Output_Get:double)
   Set IsBackblingOnAlpha(IsBackblingOnAlpha="To Float (Boolean)"(InBool=@n48))
  ]}
]
variables [
 AnimDyn_Off {type=number}
 Backward {type=number}
 BaseBall_Slide {type=boolean}
 BwdMovement {type=number}
 Cape Alpha {type=number}
 ChaosClothingInteractor {type=ChaosClothingInteractor, default=null}
 Cloth Core Settings_0 {
  type=ClothCoreSettingStruct
  default {
   collisionThickness=0
   dampingCoefficient=0.1
   frictionCoefficient=0.3
   localDampingCoefficient=0.1
   numberIterations=2
   numberSubsteps=1
   selfCollisionThickness=0
   animDriveDamping=<0 1>
   animDriveStiffness=<0 1>
   areaStiffness=<1 1>
   bendingStiffness=<1 1>
   edgeStiffness=<1 1>
   tetherStiffness=<1 1>
  }
 }
 Cloth Off {type=boolean}
 Cloth Paramaters_0 {
  type=ClothParamsStruct
  default {
   dBNO {
    angularVelocityScale=0.6
    fictitiousAngularScale=0.1
    jointName=pelvis
    dragCoefficient=<0.2 1>
    gravityOverride=<-980 400 0>
    liftCoefficient=<0 0.5>
    linearVelocityScale=<0.3 0.3 0.1>
   }
   dBNO_Carried {
    angularVelocityScale=0.6
    fictitiousAngularScale=0.1
    jointName=pelvis
    dragCoefficient=<0.2 1>
    gravityOverride=<-980 -200 0>
    liftCoefficient=<0 0.5>
    linearVelocityScale=<0.3 0.3 0.1>
   }
   dBNO_Impulse {
    angularVelocityScale=0
    fictitiousAngularScale=0
    jointName=pelvis
    dragCoefficient=<0 0.2>
    gravityOverride=<-3000 -2000 0>
    liftCoefficient=<0 0.1>
    linearVelocityScale=<0.3 0.1 0.3>
   }
   dBNO_RollBwd {
    angularVelocityScale=0
    fictitiousAngularScale=0
    jointName=pelvis
    dragCoefficient=<0 0.2>
    gravityOverride=<-2000 0 0>
    liftCoefficient=<0 0.1>
    linearVelocityScale=<0.3 0.1 0.3>
   }
   dBNO_RollFwd {
    angularVelocityScale=0
    fictitiousAngularScale=0
    jointName=pelvis
    dragCoefficient=<0 0.2>
    gravityOverride=<-980 0 0>
    liftCoefficient=<0 0.1>
    linearVelocityScale=<0.3 0.1 0.3>
   }
   dBNO_Tumbling {
    angularVelocityScale=0
    fictitiousAngularScale=0
    jointName=pelvis
    dragCoefficient=<0 0.2>
    gravityOverride=<-2500 -300 0>
    liftCoefficient=<0 0.1>
    linearVelocityScale=<0.3 0.1 0.3>
   }
   emoteOrMelee {
    angularVelocityScale=0.2
    fictitiousAngularScale=0.1
    jointName=pelvis
    dragCoefficient=<0 0.3>
    gravityOverride=<-980 -500 0>
    liftCoefficient=<0 0.2>
    linearVelocityScale=<0.05 0.05 0.05>
   }
   front_End {
    angularVelocityScale=0.1
    fictitiousAngularScale=0.2
    jointName=root
    dragCoefficient=<0.2 0.2>
    gravityOverride=<0 -120 -980>
    liftCoefficient=<0.1 0.1>
    linearVelocityScale=<0.1 0.1 0.1>
   }
   onGround_Crouch_Moving {
    angularVelocityScale=0.1
    fictitiousAngularScale=0.3
    jointName=root
    dragCoefficient=<2 2>
    gravityOverride=<0 0 -980>
    liftCoefficient=<0.6 0.6>
    linearVelocityScale=<0.6 0.6 0.5>
   }
   onGround_Falling {
    angularVelocityScale=0.6
    fictitiousAngularScale=0
    jointName=root
    dragCoefficient=<0 0.2>
    gravityOverride=<0 -500 -980>
    liftCoefficient=<0 0.1>
    linearVelocityScale=<2 1 0.3>
   }
   onGround_Moving {
    angularVelocityScale=0.6
    fictitiousAngularScale=0.1
    jointName=root
    dragCoefficient=<0 0.3>
    gravityOverride=<0 -100 -980>
    liftCoefficient=<0 0.2>
    linearVelocityScale=<0.5 1 1>
   }
   onGround_Standing {
    angularVelocityScale=0.05
    fictitiousAngularScale=0.6
    jointName=root
    dragCoefficient=<0 2>
    gravityOverride=<0 -100 -980>
    liftCoefficient=<0 0.5>
    linearVelocityScale=<0.45 1 1>
   }
   rail_Grinding {
    angularVelocityScale=0.2
    fictitiousAngularScale=0
    jointName=root
    dragCoefficient=<0.1 0.2>
    gravityOverride=<50 -500 -400>
    liftCoefficient=<0 0.5>
    linearVelocityScale=<0.1 0.1 0.1>
   }
   riding_Moving {
    angularVelocityScale=0.05
    fictitiousAngularScale=1
    jointName=root
    dragCoefficient=<0.5 0.5>
    gravityOverride=<0 0 -980>
    liftCoefficient=<0.5 0.5>
    linearVelocityScale=<0.4 0.6 0.3>
   }
   riding_Moving_Targeting {
    angularVelocityScale=0.05
    fictitiousAngularScale=0.7
    jointName=root
    dragCoefficient=<0.5 0.5>
    gravityOverride=<0 0 -980>
    liftCoefficient=<0.5 0.5>
    linearVelocityScale=<0.2 0.6 0.2>
   }
   riding_Standing {
    angularVelocityScale=0.05
    fictitiousAngularScale=1
    jointName=root
    dragCoefficient=<0.5 0.5>
    gravityOverride=<0 -450 -980>
    liftCoefficient=<0.5 0.5>
    linearVelocityScale=<0.6 0.6 0.3>
   }
   skydive_DivingDown {
    angularVelocityScale=0.1
    fictitiousAngularScale=0
    jointName=pelvis
    dragCoefficient=<0 1.5>
    gravityOverride=<-450 -100 0>
    liftCoefficient=<0 0.3>
    linearVelocityScale=<0.1 0.1 0.1>
   }
   skydive_FloatingDown {
    angularVelocityScale=0.1
    fictitiousAngularScale=0
    jointName=root
    dragCoefficient=<0 2.5>
    gravityOverride=<0 0 980>
    liftCoefficient=<0 0.4>
    linearVelocityScale=<0.12 0.2 0.1>
   }
   skydive_GoingUp {
    angularVelocityScale=0.1
    fictitiousAngularScale=0
    jointName=root
    dragCoefficient=<0 0>
    gravityOverride=<0 -3000 3000>
    liftCoefficient=<0 0>
    linearVelocityScale=<0.1 0.1 0.05>
   }
   skydive_Parachuting {
    angularVelocityScale=0.1
    fictitiousAngularScale=0
    jointName=root
    dragCoefficient=<0 2.5>
    gravityOverride=<0 0 -980>
    liftCoefficient=<0 0.5>
    linearVelocityScale=<0.2 0.2 0.2>
   }
   skydive_Surfing {
    angularVelocityScale=0.6
    fictitiousAngularScale=0.1
    jointName=root
    dragCoefficient=<1.5 1.5>
    gravityOverride=<0 -600 -500>
    liftCoefficient=<1 1>
    linearVelocityScale=<0.1 0.1 0.3>
   }
   swimming {
    angularVelocityScale=0.6
    fictitiousAngularScale=0
    jointName=root
    dragCoefficient=<0 2>
    gravityOverride=<0 -600 400>
    liftCoefficient=<0 0.4>
    linearVelocityScale=<0.6 0.2 0.2>
   }
   tactical_Sprint {
    angularVelocityScale=0.6
    fictitiousAngularScale=0.1
    jointName=root
    dragCoefficient=<2 2>
    gravityOverride=<0 0 -980>
    liftCoefficient=<0.5 0.5>
    linearVelocityScale=<0.6 0.6 0.5>
   }
   ziplining {
    angularVelocityScale=0.6
    fictitiousAngularScale=0
    jointName=root
    dragCoefficient=<0 1.5>
    gravityOverride=<0 0 -980>
    liftCoefficient=<0 0.3>
    linearVelocityScale=<0.4 0.4 0.5>
   }
  }
 }
 Crouch {type=number}
 Crouch Off {type=boolean}
 CustomAnim_Off {type=boolean}
 DeltaTimeLocal {type=number}
 emote_alpha {type=number}
 floor_slope {type=Rotator}
 Forward {type=number}
 Gravity Override Params {type=GravityOverrideParamsStruct}
 GravityOverride {type=Vector}
 Gust Local Settings {type=WindGustLocalSettingStruct, default {elapsedTime=0, finalGustStrengthLocal=0, gustDuration=0, gustStrength=0, timeBetweenGusts=0}}
 Is Backbling A Cape {type=boolean}
 IsBackblingOn {type=boolean}
 IsBackblingOnAlpha {type=number}
 Jumping {type=number}
 Jumps {type=boolean}
 Knee_Slide {type=boolean}
 Left {type=number}
 Left_Calf {type=number}
 Melee {type=number}
 moving {type=number}
 RBAN Off {type=boolean}
 resetClothLandingFromZilpine {type=boolean}
 Right {type=number}
 TrailControllerRot {type=Rotator}
 Wind Gust Params {
  type=WindGustParamsStruct
  default {
   dBNO {
    maxDownInterpSpeed=5
    maxGustDuration=2
    maxGustStrength=7
    maxTimeBetweenGusts=7
    maxUpInterpSpeed=5
    minDownInterpSpeed=2
    minGustDuration=0.5
    minGustStrength=3
    minTimeBetweenGusts=0
    minUpInterpSpeed=2
    windTurbulenceScalarA=0.1
    windTurbulenceScalarB=1
    windAmplitude=<0 0 0>
    windFrequency=<0 0 0>
    windOffset=<0 0 0>
   }
   dBNO_Carried {
    maxDownInterpSpeed=5
    maxGustDuration=2
    maxGustStrength=7
    maxTimeBetweenGusts=7
    maxUpInterpSpeed=5
    minDownInterpSpeed=2
    minGustDuration=0.5
    minGustStrength=3
    minTimeBetweenGusts=0
    minUpInterpSpeed=2
    windTurbulenceScalarA=0.1
    windTurbulenceScalarB=1
    windAmplitude=<0 0 0>
    windFrequency=<0 0 0>
    windOffset=<0 0 0>
   }
   dBNO_Impulse {
    maxDownInterpSpeed=5
    maxGustDuration=2
    maxGustStrength=7
    maxTimeBetweenGusts=7
    maxUpInterpSpeed=5
    minDownInterpSpeed=2
    minGustDuration=0.5
    minGustStrength=3
    minTimeBetweenGusts=0
    minUpInterpSpeed=2
    windTurbulenceScalarA=0.1
    windTurbulenceScalarB=1
    windAmplitude=<0 0 0>
    windFrequency=<0 0 0>
    windOffset=<0 0 0>
   }
   dBNO_RollBwd {
    maxDownInterpSpeed=5
    maxGustDuration=2
    maxGustStrength=7
    maxTimeBetweenGusts=7
    maxUpInterpSpeed=5
    minDownInterpSpeed=2
    minGustDuration=0.5
    minGustStrength=3
    minTimeBetweenGusts=0
    minUpInterpSpeed=2
    windTurbulenceScalarA=0.1
    windTurbulenceScalarB=1
    windAmplitude=<0 0 0>
    windFrequency=<0 0 0>
    windOffset=<0 0 0>
   }
   dBNO_RollFwd {
    maxDownInterpSpeed=5
    maxGustDuration=2
    maxGustStrength=7
    maxTimeBetweenGusts=7
    maxUpInterpSpeed=5
    minDownInterpSpeed=2
    minGustDuration=0.5
    minGustStrength=3
    minTimeBetweenGusts=0
    minUpInterpSpeed=2
    windTurbulenceScalarA=0.1
    windTurbulenceScalarB=1
    windAmplitude=<0 0 0>
    windFrequency=<0 0 0>
    windOffset=<0 0 0>
   }
   dBNO_Tumbling {
    maxDownInterpSpeed=5
    maxGustDuration=2
    maxGustStrength=7
    maxTimeBetweenGusts=7
    maxUpInterpSpeed=5
    minDownInterpSpeed=2
    minGustDuration=0.5
    minGustStrength=3
    minTimeBetweenGusts=0
    minUpInterpSpeed=2
    windTurbulenceScalarA=0.1
    windTurbulenceScalarB=1
    windAmplitude=<0 0 0>
    windFrequency=<0 0 0>
    windOffset=<0 0 0>
   }
   emoteOrMelee {
    maxDownInterpSpeed=5
    maxGustDuration=2
    maxGustStrength=7
    maxTimeBetweenGusts=7
    maxUpInterpSpeed=5
    minDownInterpSpeed=2
    minGustDuration=0.5
    minGustStrength=3
    minTimeBetweenGusts=0
    minUpInterpSpeed=2
    windTurbulenceScalarA=0.1
    windTurbulenceScalarB=1
    windAmplitude=<30 25 50>
    windFrequency=<2 3 5>
    windOffset=<-300 -150 0>
   }
   front_End {
    maxDownInterpSpeed=5
    maxGustDuration=2
    maxGustStrength=4
    maxTimeBetweenGusts=7
    maxUpInterpSpeed=5
    minDownInterpSpeed=2
    minGustDuration=0.5
    minGustStrength=2
    minTimeBetweenGusts=0
    minUpInterpSpeed=2
    windTurbulenceScalarA=0.1
    windTurbulenceScalarB=1
    windAmplitude=<0.05 0.05 0.05>
    windFrequency=<0.1 0.02 0.05>
    windOffset=<0 0 0>
   }
   onGround_Crouch_Moving {
    maxDownInterpSpeed=5
    maxGustDuration=2
    maxGustStrength=7
    maxTimeBetweenGusts=7
    maxUpInterpSpeed=5
    minDownInterpSpeed=2
    minGustDuration=0.5
    minGustStrength=3
    minTimeBetweenGusts=0
    minUpInterpSpeed=2
    windTurbulenceScalarA=0.1
    windTurbulenceScalarB=1
    windAmplitude=<100 25 75>
    windFrequency=<1.5 0.5 7.5>
    windOffset=<0 -100 -175>
   }
   onGround_Falling {
    maxDownInterpSpeed=5
    maxGustDuration=2
    maxGustStrength=7
    maxTimeBetweenGusts=7
    maxUpInterpSpeed=5
    minDownInterpSpeed=2
    minGustDuration=0.5
    minGustStrength=3
    minTimeBetweenGusts=0
    minUpInterpSpeed=2
    windTurbulenceScalarA=0.1
    windTurbulenceScalarB=1
    windAmplitude=<0 0 0>
    windFrequency=<0 0 0>
    windOffset=<0 0 0>
   }
   onGround_Moving {
    maxDownInterpSpeed=5
    maxGustDuration=8
    maxGustStrength=7
    maxTimeBetweenGusts=8
    maxUpInterpSpeed=5
    minDownInterpSpeed=2
    minGustDuration=0.2
    minGustStrength=3
    minTimeBetweenGusts=1
    minUpInterpSpeed=2
    windTurbulenceScalarA=0.1
    windTurbulenceScalarB=0.8
    windAmplitude=<30 65 42>
    windFrequency=<5 6 8.6>
    windOffset=<30 -250 -300>
   }
   onGround_Standing {
    maxDownInterpSpeed=5
    maxGustDuration=8
    maxGustStrength=7
    maxTimeBetweenGusts=8
    maxUpInterpSpeed=5
    minDownInterpSpeed=2
    minGustDuration=0.2
    minGustStrength=2
    minTimeBetweenGusts=1
    minUpInterpSpeed=2
    windTurbulenceScalarA=0.1
    windTurbulenceScalarB=0.8
    windAmplitude=<5 5 5>
    windFrequency=<1.5 1.5 1.5>
    windOffset=<0 0 0>
   }
   rail_Grinding {
    maxDownInterpSpeed=5
    maxGustDuration=8
    maxGustStrength=7
    maxTimeBetweenGusts=8
    maxUpInterpSpeed=5
    minDownInterpSpeed=2
    minGustDuration=0.2
    minGustStrength=3
    minTimeBetweenGusts=1
    minUpInterpSpeed=2
    windTurbulenceScalarA=0.1
    windTurbulenceScalarB=0.8
    windAmplitude=<20 65 42>
    windFrequency=<5 6 8.6>
    windOffset=<50 -500 -400>
   }
   riding_Moving {
    maxDownInterpSpeed=5
    maxGustDuration=2
    maxGustStrength=7
    maxTimeBetweenGusts=7
    maxUpInterpSpeed=5
    minDownInterpSpeed=2
    minGustDuration=0.5
    minGustStrength=3
    minTimeBetweenGusts=0
    minUpInterpSpeed=2
    windTurbulenceScalarA=0.1
    windTurbulenceScalarB=1
    windAmplitude=<0 0 0>
    windFrequency=<0 0 0>
    windOffset=<0 0 0>
   }
   riding_Moving_Targeting {
    maxDownInterpSpeed=5
    maxGustDuration=2
    maxGustStrength=7
    maxTimeBetweenGusts=7
    maxUpInterpSpeed=5
    minDownInterpSpeed=2
    minGustDuration=0.5
    minGustStrength=3
    minTimeBetweenGusts=0
    minUpInterpSpeed=2
    windTurbulenceScalarA=0.1
    windTurbulenceScalarB=1
    windAmplitude=<0 0 0>
    windFrequency=<0 0 0>
    windOffset=<0 0 0>
   }
   riding_Standing {
    maxDownInterpSpeed=5
    maxGustDuration=2
    maxGustStrength=7
    maxTimeBetweenGusts=7
    maxUpInterpSpeed=5
    minDownInterpSpeed=2
    minGustDuration=0.5
    minGustStrength=3
    minTimeBetweenGusts=0
    minUpInterpSpeed=2
    windTurbulenceScalarA=0.1
    windTurbulenceScalarB=1
    windAmplitude=<0 0 0>
    windFrequency=<0 0 0>
    windOffset=<0 0 0>
   }
   skydive_DivingDown {
    maxDownInterpSpeed=5
    maxGustDuration=8
    maxGustStrength=7
    maxTimeBetweenGusts=8
    maxUpInterpSpeed=5
    minDownInterpSpeed=2
    minGustDuration=0.2
    minGustStrength=3
    minTimeBetweenGusts=1
    minUpInterpSpeed=2
    windTurbulenceScalarA=0.1
    windTurbulenceScalarB=4
    windAmplitude=<0 25 35>
    windFrequency=<0 4 10>
    windOffset=<-450 -100 100>
   }
   skydive_FloatingDown {
    maxDownInterpSpeed=5
    maxGustDuration=8
    maxGustStrength=1
    maxTimeBetweenGusts=8
    maxUpInterpSpeed=5
    minDownInterpSpeed=2
    minGustDuration=0.2
    minGustStrength=0
    minTimeBetweenGusts=1
    minUpInterpSpeed=2
    windTurbulenceScalarA=0.5
    windTurbulenceScalarB=1
    windAmplitude=<50 20 35>
    windFrequency=<8 2 3>
    windOffset=<0 -300 200>
   }
   skydive_GoingUp {
    maxDownInterpSpeed=5
    maxGustDuration=8
    maxGustStrength=7
    maxTimeBetweenGusts=8
    maxUpInterpSpeed=5
    minDownInterpSpeed=2
    minGustDuration=0.2
    minGustStrength=3
    minTimeBetweenGusts=1
    minUpInterpSpeed=2
    windTurbulenceScalarA=0.1
    windTurbulenceScalarB=0.8
    windAmplitude=<0 0 0>
    windFrequency=<0 0 0>
    windOffset=<0 -500 300>
   }
   skydive_Parachuting {
    maxDownInterpSpeed=5
    maxGustDuration=8
    maxGustStrength=1
    maxTimeBetweenGusts=8
    maxUpInterpSpeed=5
    minDownInterpSpeed=2
    minGustDuration=5
    minGustStrength=0
    minTimeBetweenGusts=5
    minUpInterpSpeed=2
    windTurbulenceScalarA=0.2
    windTurbulenceScalarB=1
    windAmplitude=<25 50 75>
    windFrequency=<8 6 2>
    windOffset=<-150 -400 -150>
   }
   skydive_Surfing {
    maxDownInterpSpeed=5
    maxGustDuration=2
    maxGustStrength=7
    maxTimeBetweenGusts=7
    maxUpInterpSpeed=5
    minDownInterpSpeed=2
    minGustDuration=0.5
    minGustStrength=3
    minTimeBetweenGusts=0
    minUpInterpSpeed=2
    windTurbulenceScalarA=0.1
    windTurbulenceScalarB=1
    windAmplitude=<0 0 0>
    windFrequency=<0 0 0>
    windOffset=<0 0 0>
   }
   swimming {
    maxDownInterpSpeed=5
    maxGustDuration=2
    maxGustStrength=7
    maxTimeBetweenGusts=7
    maxUpInterpSpeed=5
    minDownInterpSpeed=2
    minGustDuration=0.5
    minGustStrength=3
    minTimeBetweenGusts=0
    minUpInterpSpeed=2
    windTurbulenceScalarA=0.1
    windTurbulenceScalarB=1
    windAmplitude=<10 25 20>
    windFrequency=<1.5 0.5 7.5>
    windOffset=<0 -150 50>
   }
   tactical_Sprint {
    maxDownInterpSpeed=5
    maxGustDuration=2
    maxGustStrength=7
    maxTimeBetweenGusts=7
    maxUpInterpSpeed=5
    minDownInterpSpeed=2
    minGustDuration=0.5
    minGustStrength=3
    minTimeBetweenGusts=0
    minUpInterpSpeed=2
    windTurbulenceScalarA=0.1
    windTurbulenceScalarB=1
    windAmplitude=<0 0 0>
    windFrequency=<0 0 0>
    windOffset=<0 0 0>
   }
   ziplining {
    maxDownInterpSpeed=5
    maxGustDuration=2
    maxGustStrength=7
    maxTimeBetweenGusts=7
    maxUpInterpSpeed=5
    minDownInterpSpeed=2
    minGustDuration=0.5
    minGustStrength=3
    minTimeBetweenGusts=0
    minUpInterpSpeed=2
    windTurbulenceScalarA=0.1
    windTurbulenceScalarB=1
    windAmplitude=<0 0 0>
    windFrequency=<0 0 0>
    windOffset=<0 0 0>
   }
  }
 }
]
