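# Control Rig for the F_MED_Silencer face accessory part that applies corrective offsets to lip, hair and glasses bones.
# A spherical pose reader tracks the head relative to neck_01, with debug drawing left on.
# Its smoothed output weights rotation and translation offsets on selected left and right hair bones as the head enters that pose region.
# Lips, glasses and one left-front hair bone also receive constant full-weight adjustments.
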ControlRigBlueprint /BRCosmetics/Characters/Player/Female/Medium/Bodies/F_MED_Silencer/Meshes/Parts/F_MED_Silencer_FaceAcc_CtrlRig
graphs [
 RigVMFunctionLibrary {links [], nodes []}
 RigVMModel {nodes [
   RigUnit_BeginExecution = &RigUnit_BeginExecution()
   &RigUnit_ModifyTransforms(
    ItemToModify=((Item=(Type=Bone,Name="Right_Lip"),Transform=(Rotation=(X=0,Y=0,Z=0,W=1),Translation=(X=0,Y=-0.26,Z=0),Scale3D=(X=1,Y=1,Z=1))))
    Weight=1
    ExecutePin=@RigUnit_BeginExecution.ExecutePin
   )
   > &RigUnit_ModifyTransforms(
    ItemToModify=((Item=(Type=Bone,Name="Left_Lip"),Transform=(Rotation=(X=0,Y=0,Z=0,W=1),Translation=(X=0,Y=-0.26,Z=0),Scale3D=(X=1,Y=1,Z=1))))
    Weight=1
   )
   > n1 = &RigUnit_SphericalPoseReader(
    ActiveRegionScaleFactors=(PositiveWidth=0.7,NegativeWidth=0.2,PositiveHeight=0.7,NegativeHeight=0.5)
    ActiveRegionSize=0.2
    Debug=(bDrawDebug=true,DebugScale=25,DebugSegments=20,DebugThickness=0.25)
    DriverAxis=(X=1,Y=0,Z=0)
    DriverItem=(Type=Bone,Name="head")
    FalloffRegionScaleFactors=(PositiveWidth=0.4,NegativeWidth=0.1,PositiveHeight=0.4,NegativeHeight=0.2)
    FalloffSize=0.45
    OptionalParentItem=(Type=Bone,Name="neck_01")
    RotationOffset=(X=-90,Y=0,Z=30)
   )
   AlphaInterp = &RigVMFunction_AlphaInterp(InterpSpeedDecreasing=10, InterpSpeedIncreasing=10, Scale=1, bClampResult=False, Value=@n1.OutputParam)
   > &RigUnit_ModifyTransforms(
    ItemToModify=((Item=(Type=Bone,Name="dyn_hair_l_fr_side_01"),Transform=(Rotation=(X=0,Y=0,Z=0,W=1),Translation=(X=0,Y=0,Z=0),Scale3D=(X=1,Y=1,Z=1))))
    Weight=@AlphaInterp.Result
    ItemToModify.0.Transform.Rotation=&RigVMFunction_MathQuaternionFromEuler(Euler=(X=5,Y=-10,Z=0)).Result
   )
   > &RigUnit_ModifyTransforms(
    ItemToModify=((Item=(Type=Bone,Name="dyn_hair_l_fr_side_02"),Transform=(Rotation=(X=0,Y=0,Z=0,W=1),Translation=(X=0,Y=0,Z=0),Scale3D=(X=1,Y=1,Z=1))))
    Weight=@AlphaInterp.Result
    ItemToModify.0.Transform.Rotation=&RigVMFunction_MathQuaternionFromEuler(Euler=(X=-10,Y=-5,Z=5)).Result
   )
   > &RigUnit_ModifyTransforms(
    ItemToModify=((Item=(Type=Bone,Name="dyn_hair_l_fr_01"),Transform=(Rotation=(X=0,Y=0,Z=0,W=1),Translation=(X=0,Y=0,Z=0),Scale3D=(X=1,Y=1,Z=1))))
    Weight=@AlphaInterp.Result
    ItemToModify.0.Transform.Rotation=&RigVMFunction_MathQuaternionFromEuler(Euler=(X=5,Y=-30,Z=0)).Result
   )
   > &RigUnit_ModifyTransforms(
    ItemToModify=((Item=(Type=Bone,Name="dyn_hair_r_bk_01"),Transform=(Rotation=(X=0,Y=0,Z=0,W=1),Translation=(X=-2,Y=0,Z=0),Scale3D=(X=1,Y=1,Z=1))))
    Weight=@AlphaInterp.Result
    ItemToModify.0.Transform.Rotation=&RigVMFunction_MathQuaternionFromEuler(Euler=(X=0,Y=50,Z=0)).Result
   )
   > &RigUnit_ModifyTransforms(
    ItemToModify=((Item=(Type=Bone,Name="dyn_hair_r_mid_01"),Transform=(Rotation=(X=0,Y=0,Z=0,W=1),Translation=(X=-2,Y=0,Z=0),Scale3D=(X=1,Y=1,Z=1))))
    Weight=@AlphaInterp.Result
    ItemToModify.0.Transform.Rotation=&RigVMFunction_MathQuaternionFromEuler(Euler=(X=0,Y=40,Z=0)).Result
   )
   > &RigUnit_ModifyTransforms(
    ItemToModify=((Item=(Type=Bone,Name="dyn_hair_l_bk_01"),Transform=(Rotation=(X=0,Y=0,Z=0,W=1),Translation=(X=0,Y=0,Z=0),Scale3D=(X=1,Y=1,Z=1))))
    Weight=@AlphaInterp.Result
    ItemToModify.0.Transform.Rotation=&RigVMFunction_MathQuaternionFromEuler(Euler=(X=0,Y=0,Z=10)).Result
   )
   > &RigUnit_ModifyTransforms(
    ItemToModify=((Item=(Type=Bone,Name="dyn_hair_r_side_01"),Transform=(Rotation=(X=0,Y=0,Z=0,W=1),Translation=(X=-1.5,Y=0,Z=0),Scale3D=(X=1,Y=1,Z=1))))
    Weight=@AlphaInterp.Result
    ItemToModify.0.Transform.Rotation=&RigVMFunction_MathQuaternionFromEuler(Euler=(X=0,Y=0,Z=15)).Result
   )
   > &RigUnit_ModifyTransforms(
    ItemToModify=((Item=(Type=Bone,Name="dyn_glasses"),Transform=(Rotation=(X=0,Y=0,Z=0,W=1),Translation=(X=-0.3,Y=0,Z=0),Scale3D=(X=1,Y=1,Z=1))))
    Weight=1
    ItemToModify.0.Transform.Rotation=&RigVMFunction_MathQuaternionFromEuler(Euler=(X=0,Y=0,Z=0)).Result
   )
   > &RigUnit_ModifyTransforms(
    ItemToModify=((Item=(Type=Bone,Name="dyn_hair_l_fr_01"),Transform=(Rotation=(X=0,Y=0,Z=0,W=1),Translation=(X=0,Y=0,Z=0),Scale3D=(X=1,Y=1,Z=1))))
    Weight=1
    ItemToModify.0.Transform.Rotation=&RigVMFunction_MathQuaternionFromEuler(Euler=(X=0,Y=0,Z=2)).Result
   )
  ]}
]
hierarchy [
 root {loc=<0 0 0>, rot=<0 0 0 1>}
 pelvis {parent=root, loc=<0 -0.5301 98.0571>, rot=<0.0127 0.707 -0.0127 -0.707>}
 spine_01 {parent=pelvis, loc=<6.1208 -0.9328 0>, rot=<0 0 0.034 -0.9994>}
 spine_02 {parent=spine_01, loc=<6.5996 -0.2095 0>, rot=<0 0 0.0238 -0.9997>}
 spine_03 {parent=spine_02, loc=<6.4914 -0.1237 0>, rot=<0 0 0.0149 -0.9999>}
 spine_04 {parent=spine_03, loc=<6.9598 0.0794 0>, rot=<0 0 -0.0132 -0.9999>}
 spine_05 {parent=spine_04, loc=<6.9987 0 0>, rot=<0 0 -0.0015 -1>}
 clavicle_l {parent=spine_05, loc=<5.4831 -0.1053 -3.2946>, rot=<0.0465 -0.7289 -0.009 -0.6829>}
 neck_01 {parent=spine_05, loc=<11.3138 0.1529 0>, rot=<0 0 -0.2003 -0.9797>}
 neck_02 {parent=neck_01, loc=<4.5627 0 0>, rot=<0 0 0.0504 -0.9987>}
 head {parent=neck_02, loc=<5.3914 0 0>, rot=<0 0 0.115 -0.9934>}
 dyn_hair_fr_c_01 {parent=head, loc=<10.4972 8.2192 1.3226>, rot=<-0.012 -0.0768 -0.1618 -0.9838>}
 dyn_hair_fr_c_02 {parent=dyn_hair_fr_c_01, loc=<2.9504 0.777 -0.5095>, rot=<-0.31 -0.7302 -0.0509 -0.6067>}
 dyn_hair_fr_c_03 {parent=dyn_hair_fr_c_02, loc=<4.7154 -1.5735 0.4396>, rot=<0.0311 0.2699 0.3485 -0.8971>}
 dyn_hair_l_fr_side_01 {parent=head, loc=<11.6321 9.7586 0.0817>, rot=<0.1499 -0.8427 0.0863 -0.5098>}
 dyn_hair_l_fr_side_02 {parent=dyn_hair_l_fr_side_01, loc=<8.2563 0 0>, rot=<0.2175 -0.4185 0.1191 -0.8737>}
 dyn_hair_l_fr_side_03 {parent=dyn_hair_l_fr_side_02, loc=<8.315 0 0>, rot=<0.2407 -0.1685 0.3235 -0.8995>}
 dyn_hair_l_fr_01 {parent=head, loc=<10.8022 10.6638 -0.3388>, rot=<-0.0952 -0.9662 -0.0254 -0.238>}
 dyn_hair_l_fr_02 {parent=dyn_hair_l_fr_01, loc=<6.2917 0 0>, rot=<0.001 -0.0792 0.2154 -0.9733>}
 dyn_hair_l_fr_03 {parent=dyn_hair_l_fr_02, loc=<3.199 0.6011 -0.1439>, rot=<-0.1802 0.3285 0.248 -0.8934>}
 dyn_hair_r_fr_01 {parent=head, loc=<11.4965 9.0543 3.3138>, rot=<0.0749 -0.9122 -0.0297 0.4019>}
 dyn_hair_r_fr_02 {parent=dyn_hair_r_fr_01, loc=<5.9545 -0.65 0.0259>, rot=<-0.0986 0.1219 0.2039 -0.9664>}
 dyn_hair_r_fr_03 {parent=dyn_hair_r_fr_02, loc=<6.2557 0 0>, rot=<-0.1316 0.3411 -0.0914 -0.9263>}
 dyn_hair_r_fr_side_01 {parent=head, loc=<12.0157 7.5976 3.3883>, rot=<0.1862 -0.8968 -0.0784 0.3937>}
 dyn_hair_r_fr_side_02 {parent=dyn_hair_r_fr_side_01, loc=<5.8726 -0.7873 -0.9101>, rot=<-0.1432 0.2392 0.1323 -0.9512>}
 dyn_hair_r_fr_side_03 {parent=dyn_hair_r_fr_side_02, loc=<4.7492 -0.3124 -0.3526>, rot=<-0.0553 0.2408 -0.1134 -0.9623>}
 dyn_hair_r_fr_side_04 {parent=dyn_hair_r_fr_side_03, loc=<3.2298 0.2482 -1.2068>, rot=<0.1119 -0.0174 -0.2212 -0.9686>}
 dyn_r_earrings {parent=head, loc=<0.3672 0.3487 6.7993>, rot=<0.0111 -0.707 0.7071 0.0072>}
 dyn_l_earrings {parent=head, loc=<0.3672 0.3487 -6.7993>, rot=<-0.0055 -0.8005 0.5992 -0.012>}
 dyn_hair_top_r {parent=head, loc=<11.9574 2.441 6.503>, rot=<0.5747 0.078 0.3952 -0.7123>}
 dyn_hair_top {parent=head, loc=<15.0992 -3.2893 0.1495>, rot=<0.5107 0.3637 -0.2152 -0.7487>}
 dyn_hair_l_side_01 {parent=head, loc=<3.0892 -5.3374 -7.4264>, rot=<0.15 -0.8735 0.3115 -0.3428>}
 dyn_hair_l_side_02 {parent=dyn_hair_l_side_01, loc=<2.9403 0 0>, rot=<-0.1218 0.139 0.0741 -0.98>}
 dyn_hair_l_bk_01 {parent=head, loc=<4.0775 -6.7681 -2.0259>, rot=<0.0479 -0.9635 0.0109 -0.2633>}
 dyn_hair_l_bk_02 {parent=dyn_hair_l_bk_01, loc=<5.1542 0 0>, rot=<-0.0335 0.0948 0.2442 -0.9645>}
 dyn_hair_l_bk_03 {parent=dyn_hair_l_bk_02, loc=<4.4582 0 0>, rot=<-0.1133 0.1968 -0.063 -0.9718>}
 dyn_hair_r_side_01 {parent=head, loc=<5.1244 -4.133 7.7008>, rot=<0.0617 -0.9941 -0.0048 0.0887>}
 dyn_hair_r_side_02 {parent=dyn_hair_r_side_01, loc=<4.4779 0 0>, rot=<-0.0451 -0.1613 -0.3734 -0.9124>}
 dyn_hair_r_mid_01 {parent=head, loc=<4.2775 -6.2949 5.8124>, rot=<0.1634 -0.8521 -0.3778 0.3233>}
 dyn_hair_r_mid_02 {parent=dyn_hair_r_mid_01, loc=<4.0797 0 0>, rot=<0.112 -0.3444 0.1376 -0.9219>}
 dyn_hair_r_bk_01 {parent=head, loc=<2.8044 -6.0224 1.3483>, rot=<-0.1639 -0.7514 -0.5859 0.2556>}
 dyn_hair_r_bk_02 {parent=dyn_hair_r_bk_01, loc=<4.1461 0 0>, rot=<0.008 -0.1445 0.0086 -0.9894>}
 faceAttach {parent=head, loc=<0.6201 0.9296 0>, rot=<0.0029 -0.7071 -0.0029 -0.7071>}
 L_cheek_inner {parent=faceAttach, loc=<4.152 6.2884 -1.9746>, rot=<-0.0706 -0.2697 0.3184 -0.906>}
 C_nose_bridge {parent=faceAttach, loc=<-0.0068 8.8185 -0.0879>, rot=<-0.2218 0 0 -0.9751>}
 R_cheek_inner {parent=faceAttach, loc=<-4.1589 6.2865 -1.9729>, rot=<-0.0706 0.2697 -0.3184 -0.906>}
 C_jaw {parent=faceAttach, loc=<0.2289 0.3269 -2.4599>, rot=<0 0 0 1>}
 C_lip_lower_mid {parent=C_jaw, loc=<-0.2417 7.2579 -1.1115>, rot=<0 0 0 -1>}
 L_lip_lower_outer {parent=C_jaw, loc=<0.9087 6.7818 -0.887>, rot=<0.013 0.0445 0.3057 -0.951>}
 R_lip_lower_outer {parent=C_jaw, loc=<-1.4031 6.7943 -0.9245>, rot=<0.013 -0.0445 -0.3057 -0.951>}
 Left_Lip {parent=C_jaw, loc=<0.8689 6.7447 -2.26>, rot=<0 0 0 1>}
 Right_Lip {parent=C_jaw, loc=<-1.2744 6.7601 -2.266>, rot=<0 0 0 1>}
 C_lip_upper_mid {parent=faceAttach, loc=<-0.0002 7.8932 -3.2451>, rot=<0.2225 0.0013 -0.0005 -0.9749>}
 L_lip_corner {parent=faceAttach, loc=<1.9086 6.6057 -3.4591>, rot=<0.003 0.008 -0.0146 -0.9999>}
 L_lip_upper_outer {parent=faceAttach, loc=<1.1276 7.2216 -3.3403>, rot=<0.0359 -0.0728 0.2627 -0.9615>}
 R_lip_corner {parent=faceAttach, loc=<-1.9651 6.645 -3.4432>, rot=<0.003 -0.008 0.0146 -0.9999>}
 R_lip_upper_outer {parent=faceAttach, loc=<-1.1369 7.2211 -3.34>, rot=<0.0359 0.0728 -0.2627 -0.9615>}
 dyn_glasses {parent=head, loc=<3.6189 -0.8049 -0.0117>, rot=<-0.6823 -0.7309 0.0095 0.01>}
 clavicle_r {parent=spine_05, loc=<5.4831 -0.1053 3.2946>, rot=<0.7289 0.0465 -0.6829 0.009>}
 FrontEnd_Idle_Bypass {type=Curve, loc=<0 0 0>, rot=<0 0 0 1>}
]
